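# Physics asset for the neck area of the ForestBath medium female player body cosmetic, setting up secondary motion for neck pieces and a necklace.
# Dynamic neck bones hang from small base bodies with tight swing and twist limits, while the necklace forms a two-link chain from its base.
# Capsules on head, neck, upper spine and upper arms provide body volumes, linked with the base bodies by free-rotation constraints.
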PhysicsAsset /BRCosmetics/Characters/Player/Female/Medium/Bodies/F_MED_ForestBath/Meshes/F_MED_ForestBath_NeckPhysics
bodies {
 base_neck_back [{shapeType=Capsule, length=0.505, radius=0.505, shapeName=base_neck_back_capsule, rotation=<rot 0 -90 0>}]
 base_neck_front_l [{shapeType=Capsule, length=0.505, radius=0.505, shapeName=base_neck_front_l_capsule, rotation=<rot 0 -90 0>}]
 base_neck_front_r [{shapeType=Capsule, length=0.505, radius=0.505, shapeName=base_neck_front_r_capsule, rotation=<rot 0 -90 0>}]
 base_neck_mid_l [{shapeType=Capsule, length=0.505, radius=0.505, shapeName=base_neck_mid_l_capsule, rotation=<rot 0 -90 0>}]
 base_neck_mid_r [{shapeType=Capsule, length=0.505, radius=0.505, shapeName=base_neck_mid_r_capsule, rotation=<rot 0 -90 0>}]
 base_neck_midback_r [{shapeType=Capsule, length=0.505, radius=0.505, shapeName=base_neck_midback_r_capsule, rotation=<rot 0 -90 0>}]
 base_necklace [{shapeType=Capsule, length=0.505, radius=0.505, shapeName=base_necklace_capsule, rotation=<rot 0 -90 0>}]
 dyn_neck_back [{shapeType=Capsule, length=6, radius=1.5, shapeName=dyn_neck_back_capsule, center=<2.3 0.3 0>}]
 dyn_neck_front_l [
  {shapeType=Capsule, length=11, radius=1.5, shapeName=dyn_neck_front_l_capsule, center=<2.4 -4.2 0>, rotation=<rot 0 -18 90>}
  {shapeType=Capsule, length=18, radius=1.5, shapeName=dyn_neck_front_l_capsule, center=<0 -4.2 0>, rotation=<rot 0 0 90>}
 ]
 dyn_neck_front_r [{shapeType=Capsule, length=11, radius=1.5, shapeName=dyn_neck_front_l_capsule, center=<2.4 -4.2 0>, rotation=<rot 0 -18 90>}]
 dyn_neck_mid_l [{shapeType=Capsule, length=11, radius=1.5, shapeName=dyn_neck_front_l_capsule, center=<1.5 -1.85 0.5>, rotation=<rot 0 0 95>}]
 dyn_neck_mid_r [{shapeType=Capsule, length=4, radius=1.5, shapeName=dyn_neck_front_l_capsule, center=<1.5 -1.7 -0.25>, rotation=<rot 0 0 80>}]
 dyn_neck_midback_r [{shapeType=Capsule, length=4, radius=1.5, shapeName=dyn_neck_midback_r_capsule, center=<1.8 0 0>, rotation=<rot 0 0 90>}]
 dyn_necklace_1 [{shapeType=Capsule, length=14, radius=4, shapeName=dyn_necklace_2_capsule, center=<4 1.3 0>, rotation=<rot 90 0 5>}]
 dyn_necklace_2 [{shapeType=Capsule, length=11, radius=4.5, shapeName=dyn_necklace_2_capsule, center=<9.5 3.2 0>}]
 head [{shapeType=Capsule, length=8, radius=7, shapeName=head_capsule, center=<2.3419 2.6562 0>, rotation=<rot 90 180 -150>}]
 neck_01 [{shapeType=Capsule, length=1, radius=6, shapeName=neck_01_capsule, center=<0 1.57 0>, rotation=<rot 90 0 0>}]
 neck_02 [{shapeType=Capsule, length=1, radius=6, shapeName=neck_02_capsule, center=<0 1.57 0>, rotation=<rot 90 0 0>}]
 spine_04 [{shapeType=Capsule, length=15.9116, radius=13, shapeName=spine_04_capsule, center=<-3.8 1.4 0>, rotation=<rot -0.2576 -90 -0.0015>}]
 spine_05 [{shapeType=Capsule, length=8, radius=10.2, shapeName=spine_05_capsule, center=<-2.8 -1.3 0>}]
 upperarm_l [{shapeType=Capsule, length=6, radius=9, shapeName=upperarm_l_capsule, center=<5 0 0>, rotation=<rot 90 0 0>}]
 upperarm_r [{shapeType=Capsule, length=6, radius=9, shapeName=upperarm_l_capsule, center=<-5 0 0>, rotation=<rot 90 0 0>}]
}
constraints [
 {bone1Name=dyn_neck_front_l, bone2Name=base_neck_front_l, swing1LimitDegrees=0, swing2LimitDegrees=20, twistLimitDegrees=1}
 {bone1Name=dyn_neck_front_r, bone2Name=base_neck_front_r, swing1LimitDegrees=0, swing2LimitDegrees=20, twistLimitDegrees=1}
 {bone1Name=dyn_neck_mid_l, bone2Name=base_neck_mid_l, swing1LimitDegrees=5, swing2LimitDegrees=15, twistLimitDegrees=1}
 {bone1Name=dyn_neck_mid_r, bone2Name=base_neck_mid_r, swing1LimitDegrees=5, swing2LimitDegrees=15, twistLimitDegrees=1}
 {bone1Name=dyn_neck_midback_r, bone2Name=base_neck_midback_r, swing1LimitDegrees=5, swing2LimitDegrees=15, twistLimitDegrees=1}
 {
  bone1Name=base_neck_front_r
  bone2Name=neck_01
  swing1LimitDegrees=45
  swing1Motion=Free
  swing2LimitDegrees=45
  swing2Motion=Free
  twistLimitDegrees=45
  twistMotion=Free
 }
 {
  bone1Name=base_neck_mid_r
  bone2Name=neck_01
  swing1LimitDegrees=45
  swing1Motion=Free
  swing2LimitDegrees=45
  swing2Motion=Free
  twistLimitDegrees=45
  twistMotion=Free
 }
 {
  bone1Name=base_neck_midback_r
  bone2Name=neck_01
  swing1LimitDegrees=45
  swing1Motion=Free
  swing2LimitDegrees=45
  swing2Motion=Free
  twistLimitDegrees=45
  twistMotion=Free
 }
 {
  bone1Name=base_neck_mid_l
  bone2Name=neck_01
  swing1LimitDegrees=45
  swing1Motion=Free
  swing2LimitDegrees=45
  swing2Motion=Free
  twistLimitDegrees=45
  twistMotion=Free
 }
 {
  bone1Name=base_neck_front_l
  bone2Name=neck_01
  swing1LimitDegrees=45
  swing1Motion=Free
  swing2LimitDegrees=45
  swing2Motion=Free
  twistLimitDegrees=45
  twistMotion=Free
 }
 {
  bone1Name=neck_02
  bone2Name=neck_01
  swing1LimitDegrees=45
  swing1Motion=Free
  swing2LimitDegrees=45
  swing2Motion=Free
  twistLimitDegrees=45
  twistMotion=Free
 }
 {
  bone1Name=head
  bone2Name=neck_02
  swing1LimitDegrees=45
  swing1Motion=Free
  swing2LimitDegrees=45
  swing2Motion=Free
  twistLimitDegrees=45
  twistMotion=Free
 }
 {bone1Name=dyn_neck_back, bone2Name=base_neck_back, swing1LimitDegrees=15, swing2LimitDegrees=5, twistLimitDegrees=1}
 {bone1Name=dyn_necklace_1, bone2Name=base_necklace, swing1LimitDegrees=20, swing2LimitDegrees=20, twistLimitDegrees=10}
 {bone1Name=dyn_necklace_2, bone2Name=dyn_necklace_1, swing1LimitDegrees=20, swing2LimitDegrees=20, twistLimitDegrees=15}
 {
  bone1Name=neck_01
  bone2Name=spine_05
  swing1LimitDegrees=45
  swing1Motion=Free
  swing2LimitDegrees=45
  swing2Motion=Free
  twistLimitDegrees=45
  twistMotion=Free
 }
 {
  bone1Name=base_neck_back
  bone2Name=spine_05
  swing1LimitDegrees=45
  swing1Motion=Free
  swing2LimitDegrees=45
  swing2Motion=Free
  twistLimitDegrees=45
  twistMotion=Free
 }
 {
  bone1Name=upperarm_l
  bone2Name=spine_05
  swing1LimitDegrees=45
  swing1Motion=Free
  swing2LimitDegrees=45
  swing2Motion=Free
  twistLimitDegrees=45
  twistMotion=Free
 }
 {
  bone1Name=upperarm_r
  bone2Name=spine_05
  swing1LimitDegrees=45
  swing1Motion=Free
  swing2LimitDegrees=45
  swing2Motion=Free
  twistLimitDegrees=45
  twistMotion=Free
 }
 {
  bone1Name=base_necklace
  bone2Name=spine_05
  swing1LimitDegrees=45
  swing1Motion=Free
  swing2LimitDegrees=45
  swing2Motion=Free
  twistLimitDegrees=45
  twistMotion=Free
 }
 {
  bone1Name=spine_05
  bone2Name=spine_04
  swing1LimitDegrees=45
  swing1Motion=Free
  swing2LimitDegrees=45
  swing2Motion=Free
  twistLimitDegrees=45
  twistMotion=Free
 }
]
