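# Physics asset for the F_MED_ForestBath female player cosmetic body, setting up capsule bodies and joints for jacket and front-bag motion.
# Six four-segment jacket chains (front, mid, back per side) hang from base bodies on spine_03, with tight twist and swing limits widening down each chain.
# The single-segment bag hangs from a pelvis-linked base.
# Pelvis, spine and leg capsules use free angular constraints, with calf_r linked to pelvis rather than thigh_r.
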PhysicsAsset /BRCosmetics/Characters/Player/Female/Medium/Bodies/F_MED_ForestBath/Meshes/F_MED_ForestBath_Physics
bodies {
 base_bagfront [{shapeType=Capsule, length=0.505, radius=0.505, shapeName=base_bagfront_capsule, rotation=<rot 0 -90 0>}]
 base_jacket_back_l [{shapeType=Capsule, length=0.505, radius=0.505, shapeName=base_jacket_back_l_capsule, rotation=<rot 0 -90 0>}]
 base_jacket_back_r [{shapeType=Capsule, length=0.505, radius=0.505, shapeName=base_jacket_back_r_capsule, rotation=<rot 0 -90 0>}]
 base_jacket_front_l [{shapeType=Capsule, length=0.505, radius=0.505, shapeName=base_jacket_front_l_capsule, rotation=<rot 0 -90 0>}]
 base_jacket_front_r [{shapeType=Capsule, length=0.505, radius=0.505, shapeName=base_jacket_front_r_capsule, rotation=<rot 0 -90 0>}]
 base_jacket_mid_l [{shapeType=Capsule, length=0.505, radius=0.505, shapeName=base_jacket_mid_l_capsule, rotation=<rot 0 -90 0>}]
 base_jacket_mid_r [{shapeType=Capsule, length=0.505, radius=0.505, shapeName=base_jacket_mid_r_capsule, rotation=<rot 0 -90 0>}]
 calf_l [{shapeType=Capsule, length=33, radius=14, shapeName=thigh_r_capsule, center=<-23 -3 0.036>, rotation=<rot 90 0 0>}]
 calf_r [{shapeType=Capsule, length=33, radius=14, shapeName=thigh_r_capsule, center=<23 3 0>, rotation=<rot 90 0 0>}]
 dyn_bagfront [{shapeType=Capsule, length=20.5, radius=3, shapeName=dyn_bagfront_capsule, center=<-1 0 0>}]
 dyn_jacket_back_l_1 [{shapeType=Capsule, length=10, radius=5, shapeName=dyn_jacket_back_l_4_capsule, center=<5 0 4>, rotation=<rot 0 0 90>}]
 dyn_jacket_back_l_2 [{shapeType=Capsule, length=11, radius=7, shapeName=dyn_jacket_back_l_4_capsule, center=<5 -0.5 4>, rotation=<rot 0 0 90>}]
 dyn_jacket_back_l_3 [{shapeType=Capsule, length=22, radius=7, shapeName=dyn_jacket_back_l_4_capsule, center=<5 -0.5 3.5>, rotation=<rot 0 0 90>}]
 dyn_jacket_back_l_4 [{shapeType=Capsule, length=22, radius=7, shapeName=dyn_jacket_back_l_4_capsule, center=<5 -0.5 3.5>, rotation=<rot 0 0 90>}]
 dyn_jacket_back_r_1 [{shapeType=Capsule, length=10, radius=5, shapeName=dyn_jacket_front_l_1_capsule, center=<6 0.5 4>, rotation=<rot 0 0 90>}]
 dyn_jacket_back_r_2 [{shapeType=Capsule, length=11, radius=7, shapeName=dyn_jacket_front_l_4_capsule, center=<5 0.5 4>, rotation=<rot 0 0 105>}]
 dyn_jacket_back_r_3 [{shapeType=Capsule, length=22, radius=7, shapeName=dyn_jacket_front_l_4_capsule, center=<5 0.5 3.5>, rotation=<rot 0 0 105>}]
 dyn_jacket_back_r_4 [{shapeType=Capsule, length=22, radius=7, shapeName=dyn_jacket_front_l_4_capsule, center=<5 0.5 3.5>, rotation=<rot 0 0 105>}]
 dyn_jacket_front_l_1 [{shapeType=Capsule, length=10, radius=5, shapeName=dyn_jacket_front_l_1_capsule, center=<6 0.5 4>, rotation=<rot 0 0 100>}]
 dyn_jacket_front_l_2 [{shapeType=Capsule, length=4, radius=7, shapeName=dyn_jacket_front_l_2_capsule, center=<5 0.5 4.5>, rotation=<rot 0 0 100>}]
 dyn_jacket_front_l_3 [{shapeType=Capsule, length=6, radius=8.3, shapeName=dyn_jacket_front_l_3_capsule, center=<5.036 0.5 5>, rotation=<rot 0 0 100>}]
 dyn_jacket_front_l_4 [{shapeType=Capsule, length=22, radius=8.5, shapeName=dyn_jacket_front_l_4_capsule, center=<5.036 0.5 5.5>, rotation=<rot 0 0 100>}]
 dyn_jacket_front_r_1 [{shapeType=Capsule, length=10, radius=5, shapeName=dyn_jacket_front_l_1_capsule, center=<6 0.5 4>, rotation=<rot 0 0 70>}]
 dyn_jacket_front_r_2 [{shapeType=Capsule, length=4, radius=7, shapeName=dyn_jacket_front_l_4_capsule, center=<5 2 7>, rotation=<rot 0 0 70>}]
 dyn_jacket_front_r_3 [{shapeType=Capsule, length=6, radius=8, shapeName=dyn_jacket_front_l_4_capsule, center=<5 4 8>, rotation=<rot 0 0 70>}]
 dyn_jacket_front_r_4 [{shapeType=Capsule, length=22, radius=8.5, shapeName=dyn_jacket_front_l_4_capsule, center=<5 3 7>, rotation=<rot 0 0 70>}]
 dyn_jacket_mid_l_1 [{shapeType=Capsule, length=10, radius=5, shapeName=dyn_jacket_front_l_1_capsule, center=<6 0.5 4>, rotation=<rot 0 0 90>}]
 dyn_jacket_mid_l_2 [{shapeType=Capsule, length=11, radius=7, shapeName=dyn_jacket_mid_l_3_capsule, center=<5 0 4>, rotation=<rot 0 0 90>}]
 dyn_jacket_mid_l_3 [{shapeType=Capsule, length=11, radius=7.5, shapeName=dyn_jacket_mid_l_3_capsule, center=<5 0 4>, rotation=<rot 0 0 90>}]
 dyn_jacket_mid_l_4 [{shapeType=Capsule, length=33, radius=8.8, shapeName=dyn_jacket_mid_l_4_capsule, center=<5.036 0 4>, rotation=<rot 0 0 90>}]
 dyn_jacket_mid_r_1 [{shapeType=Capsule, length=10, radius=5, shapeName=dyn_jacket_front_l_1_capsule, center=<6 0.5 4>, rotation=<rot 0 0 90>}]
 dyn_jacket_mid_r_2 [{shapeType=Capsule, length=11, radius=7, shapeName=dyn_jacket_front_l_4_capsule, center=<5 0 4>, rotation=<rot 0 0 105>}]
 dyn_jacket_mid_r_3 [{shapeType=Capsule, length=11, radius=7, shapeName=dyn_jacket_front_l_4_capsule, center=<5 0 4>, rotation=<rot 0 0 105>}]
 dyn_jacket_mid_r_4 [{shapeType=Capsule, length=33, radius=8.8, shapeName=dyn_jacket_front_l_4_capsule, center=<5 0 4>, rotation=<rot 0 0 105>}]
 pelvis [{shapeType=Capsule, length=11, radius=11, shapeName=pelvis_capsule, center=<-0.9998 0 0>}]
 spine_01 [{shapeType=Capsule, length=1, radius=10.5, shapeName=spine_01_capsule, rotation=<rot 90 0 0>}]
 spine_02 [{shapeType=Capsule, length=1, radius=10.5, shapeName=spine_02_capsule, rotation=<rot 90 0 0>}]
 spine_03 [{shapeType=Capsule, length=1, radius=10.5, shapeName=spine_03_capsule, rotation=<rot 90 0 0>}]
 thigh_l [{shapeType=Capsule, length=33, radius=11, shapeName=thigh_r_capsule, center=<-25 -1.5 1.3>, rotation=<rot 100 0 0>}]
 thigh_r [{shapeType=Capsule, length=33, radius=10.5, shapeName=thigh_r_capsule, center=<25 1.5 0>, rotation=<rot 95 0 0>}]
}
constraints [
 {bone1Name=dyn_jacket_back_l_1, bone2Name=base_jacket_back_l, swing1LimitDegrees=3, swing2LimitDegrees=8, twistLimitDegrees=1}
 {bone1Name=dyn_jacket_back_l_2, bone2Name=dyn_jacket_back_l_1, swing1LimitDegrees=5, swing2LimitDegrees=15, twistLimitDegrees=1}
 {bone1Name=dyn_jacket_back_l_3, bone2Name=dyn_jacket_back_l_2, swing1LimitDegrees=8, swing2LimitDegrees=33, twistLimitDegrees=1}
 {bone1Name=dyn_jacket_back_l_4, bone2Name=dyn_jacket_back_l_3, swing1LimitDegrees=10, swing2LimitDegrees=40, twistLimitDegrees=1}
 {bone1Name=dyn_jacket_back_r_1, bone2Name=base_jacket_back_r, swing1LimitDegrees=3, swing2LimitDegrees=8, twistLimitDegrees=1}
 {bone1Name=dyn_jacket_back_r_2, bone2Name=dyn_jacket_back_r_1, swing1LimitDegrees=5, swing2LimitDegrees=15, twistLimitDegrees=1}
 {bone1Name=dyn_jacket_back_r_3, bone2Name=dyn_jacket_back_r_2, swing1LimitDegrees=8, swing2LimitDegrees=33, twistLimitDegrees=1}
 {bone1Name=dyn_jacket_back_r_4, bone2Name=dyn_jacket_back_r_3, swing1LimitDegrees=10, swing2LimitDegrees=40, twistLimitDegrees=1}
 {bone1Name=dyn_jacket_front_l_1, bone2Name=base_jacket_front_l, swing1LimitDegrees=3, swing2LimitDegrees=10, twistLimitDegrees=1}
 {bone1Name=dyn_jacket_front_l_2, bone2Name=dyn_jacket_front_l_1, swing1LimitDegrees=5, swing2LimitDegrees=20, twistLimitDegrees=1}
 {bone1Name=dyn_jacket_front_l_3, bone2Name=dyn_jacket_front_l_2, swing1LimitDegrees=8, swing2LimitDegrees=30, twistLimitDegrees=1}
 {bone1Name=dyn_jacket_front_l_4, bone2Name=dyn_jacket_front_l_3, swing1LimitDegrees=10, swing2LimitDegrees=40, twistLimitDegrees=1}
 {bone1Name=dyn_jacket_front_r_1, bone2Name=base_jacket_front_r, swing1LimitDegrees=3, swing2LimitDegrees=10, twistLimitDegrees=1}
 {bone1Name=dyn_jacket_front_r_2, bone2Name=dyn_jacket_front_r_1, swing1LimitDegrees=5, swing2LimitDegrees=20, twistLimitDegrees=1}
 {bone1Name=dyn_jacket_front_r_3, bone2Name=dyn_jacket_front_r_2, swing1LimitDegrees=8, swing2LimitDegrees=30, twistLimitDegrees=1}
 {bone1Name=dyn_jacket_front_r_4, bone2Name=dyn_jacket_front_r_3, swing1LimitDegrees=10, swing2LimitDegrees=40, twistLimitDegrees=1}
 {bone1Name=dyn_jacket_mid_l_1, bone2Name=base_jacket_mid_l, swing1LimitDegrees=3, swing2LimitDegrees=5, twistLimitDegrees=1}
 {bone1Name=dyn_jacket_mid_l_2, bone2Name=dyn_jacket_mid_l_1, swing1LimitDegrees=5, swing2LimitDegrees=10, twistLimitDegrees=1}
 {bone1Name=dyn_jacket_mid_l_3, bone2Name=dyn_jacket_mid_l_2, swing1LimitDegrees=10, swing2LimitDegrees=15, twistLimitDegrees=1}
 {bone1Name=dyn_jacket_mid_l_4, bone2Name=dyn_jacket_mid_l_3, swing1LimitDegrees=20, swing2LimitDegrees=20, twistLimitDegrees=1}
 {bone1Name=dyn_jacket_mid_r_1, bone2Name=base_jacket_mid_r, swing1LimitDegrees=3, swing2LimitDegrees=5, twistLimitDegrees=1}
 {bone1Name=dyn_jacket_mid_r_2, bone2Name=dyn_jacket_mid_r_1, swing1LimitDegrees=5, swing2LimitDegrees=10, twistLimitDegrees=1}
 {bone1Name=dyn_jacket_mid_r_3, bone2Name=dyn_jacket_mid_r_2, swing1LimitDegrees=10, swing2LimitDegrees=15, twistLimitDegrees=1}
 {bone1Name=dyn_jacket_mid_r_4, bone2Name=dyn_jacket_mid_r_3, swing1LimitDegrees=20, swing2LimitDegrees=20, twistLimitDegrees=1}
 {bone1Name=dyn_bagfront, bone2Name=base_bagfront, swing1LimitDegrees=10, swing2LimitDegrees=10, twistLimitDegrees=20}
 {
  bone1Name=calf_l
  bone2Name=thigh_l
  swing1LimitDegrees=45
  swing1Motion=Free
  swing2LimitDegrees=45
  swing2Motion=Free
  twistLimitDegrees=45
  twistMotion=Free
 }
 {
  bone1Name=base_jacket_front_l
  bone2Name=spine_03
  swing1LimitDegrees=45
  swing1Motion=Free
  swing2LimitDegrees=45
  swing2Motion=Free
  twistLimitDegrees=45
  twistMotion=Free
 }
 {
  bone1Name=base_jacket_mid_l
  bone2Name=spine_03
  swing1LimitDegrees=45
  swing1Motion=Free
  swing2LimitDegrees=45
  swing2Motion=Free
  twistLimitDegrees=45
  twistMotion=Free
 }
 {
  bone1Name=base_jacket_back_l
  bone2Name=spine_03
  swing1LimitDegrees=45
  swing1Motion=Free
  swing2LimitDegrees=45
  swing2Motion=Free
  twistLimitDegrees=45
  twistMotion=Free
 }
 {
  bone1Name=base_jacket_front_r
  bone2Name=spine_03
  swing1LimitDegrees=45
  swing1Motion=Free
  swing2LimitDegrees=45
  swing2Motion=Free
  twistLimitDegrees=45
  twistMotion=Free
 }
 {
  bone1Name=base_jacket_mid_r
  bone2Name=spine_03
  swing1LimitDegrees=45
  swing1Motion=Free
  swing2LimitDegrees=45
  swing2Motion=Free
  twistLimitDegrees=45
  twistMotion=Free
 }
 {
  bone1Name=base_jacket_back_r
  bone2Name=spine_03
  swing1LimitDegrees=45
  swing1Motion=Free
  swing2LimitDegrees=45
  swing2Motion=Free
  twistLimitDegrees=45
  twistMotion=Free
 }
 {
  bone1Name=spine_02
  bone2Name=spine_01
  swing1LimitDegrees=45
  swing1Motion=Free
  swing2LimitDegrees=45
  swing2Motion=Free
  twistLimitDegrees=45
  twistMotion=Free
 }
 {
  bone1Name=spine_03
  bone2Name=spine_02
  swing1LimitDegrees=45
  swing1Motion=Free
  swing2LimitDegrees=45
  swing2Motion=Free
  twistLimitDegrees=45
  twistMotion=Free
 }
 {
  bone1Name=spine_01
  bone2Name=pelvis
  swing1LimitDegrees=45
  swing1Motion=Free
  swing2LimitDegrees=45
  swing2Motion=Free
  twistLimitDegrees=45
  twistMotion=Free
 }
 {
  bone1Name=thigh_l
  bone2Name=pelvis
  swing1LimitDegrees=45
  swing1Motion=Free
  swing2LimitDegrees=45
  swing2Motion=Free
  twistLimitDegrees=45
  twistMotion=Free
 }
 {
  bone1Name=thigh_r
  bone2Name=pelvis
  swing1LimitDegrees=45
  swing1Motion=Free
  swing2LimitDegrees=45
  swing2Motion=Free
  twistLimitDegrees=45
  twistMotion=Free
 }
 {
  bone1Name=calf_r
  bone2Name=pelvis
  swing1LimitDegrees=45
  swing1Motion=Free
  swing2LimitDegrees=45
  swing2Motion=Free
  twistLimitDegrees=45
  twistMotion=Free
 }
 {
  bone1Name=base_bagfront
  bone2Name=pelvis
  swing1LimitDegrees=45
  swing1Motion=Free
  swing2LimitDegrees=45
  swing2Motion=Free
  twistLimitDegrees=45
  twistMotion=Free
 }
]
