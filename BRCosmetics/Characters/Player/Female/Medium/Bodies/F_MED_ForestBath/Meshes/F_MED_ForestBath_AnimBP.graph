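# Post-process-style AnimBlueprint for the F_MED_ForestBath female outfit, animating jacket, collar, necklace, furballs and pants.
# It layers copied body and face poses, then applies weighted constraints, AnimDynamics, a state-fed Control Rig and additive bone offsets, some parachute-driven.
# That cached pose feeds two gravity-overridden RigidBody simulations, then jacket and necklace corrections with look-at aiming.
# ASWon blends between this and a simpler jacket-offset pose that skips both simulations.
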AnimBlueprint /BRCosmetics/Characters/Player/Female/Medium/Bodies/F_MED_ForestBath/Meshes/F_MED_ForestBath_AnimBP
target_skeleton=&/F_MED_ForestBath_Skeleton
graphs [
 AnimGraph {nodes [
   n1 = Get GravityOverride()
   n2 = RigidBody!(ComponentPose=Local To Component(LocalPose=Use cached pose 'CACHE_Start'()), Alpha, OverrideWorldGravity:Vector=@n1, ExternalForce) -> (Pose) {
    bClampLinearTranslationLimitToRefPose=true
    bOverrideWorldGravity=true
    overridePhysicsAsset=&/F_MED_ForestBath_Physics
    simulationSpace=BaseBoneSpace
    alphaBoolBlend {blendOption=Linear}
    baseBoneRef {boneName=pelvis}
    componentAppliedLinearAccClamp=<3000 3000 3000>
    componentLinearAccScale=<0.6 0.6 0.6>
    componentLinearVelScale=<0.6 0.6 0.6>
    simSpaceSettings {
     maxAngularAcceleration=1000
     maxAngularVelocity=3
     maxLinearAcceleration=700
     maxLinearVelocity=500
     velocityScaleZ=0
     worldAlpha=0.2
     externalLinearDragV=<0.2 0.2>
    }
   }
   n3 = RigidBody(ComponentPose=@n2, OverrideWorldGravity=@n1) {
    bClampLinearTranslationLimitToRefPose=true
    bOverrideWorldGravity=true
    overridePhysicsAsset=&/F_MED_ForestBath_NeckPhysics
    simulationSpace=BaseBoneSpace
    alphaBoolBlend {blendOption=Linear}
    baseBoneRef {boneName=spine_05}
    componentAppliedLinearAccClamp=<3000 3000 3000>
    componentLinearAccScale=<0.6 0.6 0.6>
    componentLinearVelScale=<0.6 0.6 0.6>
    simSpaceSettings {
     maxAngularAcceleration=1000
     maxAngularVelocity=3
     maxLinearAcceleration=700
     maxLinearVelocity=500
     velocityScaleZ=0
     worldAlpha=0.05
     externalLinearDragV=<0.2 0.2>
    }
   }
   n4 = "Transform (Modify) Bone - Bone: base_jacket_front_l"!(ComponentPose=@n3, bAlphaBoolEnabled:bool=True, Translation="0,0,0", Rotation="3,0,0", Scale="1,1,1") -> (Pose) {
    alphaInputType=Bool
    rotationMode=BMM_Additive
    rotationSpace=BCS_BoneSpace
    scaleMode=BMM_Additive
    scaleSpace=BCS_BoneSpace
    translationMode=BMM_Additive
    translationSpace=BCS_BoneSpace
    alphaBoolBlend {blendInTime=0.5, blendOption=Sinusoidal, blendOutTime=1}
    boneToModify {boneName=base_jacket_front_l}
    rotation {pitch=3}
   }
   n5 = "Transform (Modify) Bone - Bone: base_jacket_front_r"!(ComponentPose=@n4, bAlphaBoolEnabled:bool=True, Translation="0,0,0", Rotation="3,0,0", Scale="1,1,1") -> (Pose) {
    alphaInputType=Bool
    rotationMode=BMM_Additive
    rotationSpace=BCS_BoneSpace
    scaleMode=BMM_Additive
    scaleSpace=BCS_BoneSpace
    translationMode=BMM_Additive
    translationSpace=BCS_BoneSpace
    alphaBoolBlend {blendInTime=0.5, blendOption=Sinusoidal, blendOutTime=1}
    boneToModify {boneName=base_jacket_front_r}
    rotation {pitch=3}
   }
   n6 = "Transform (Modify) Bone - Bone: base_jacket_front_l"(ComponentPose=@n5, bAlphaBoolEnabled=True, Translation="0,0,0", Rotation="15,-5,0", Scale="1,1,1") {
    alphaInputType=Bool
    rotationMode=BMM_Additive
    rotationSpace=BCS_BoneSpace
    scaleMode=BMM_Additive
    scaleSpace=BCS_BoneSpace
    translationMode=BMM_Additive
    translationSpace=BCS_BoneSpace
    alphaBoolBlend {blendInTime=0.5, blendOption=Sinusoidal, blendOutTime=0.5}
    boneToModify {boneName=base_jacket_front_l}
    rotation {pitch=15, yaw=-5}
   }
   n7 = "Transform (Modify) Bone - Bone: base_jacket_front_l"(ComponentPose=@n6, bAlphaBoolEnabled=True, Translation="0,0,0", Rotation="15,-5,0", Scale="1,1,1") {
    alphaInputType=Bool
    rotationMode=BMM_Additive
    rotationSpace=BCS_BoneSpace
    scaleMode=BMM_Additive
    scaleSpace=BCS_BoneSpace
    translationMode=BMM_Additive
    translationSpace=BCS_BoneSpace
    alphaBoolBlend {blendInTime=0.5, blendOption=Sinusoidal, blendOutTime=0.5}
    boneToModify {boneName=base_jacket_front_l}
    rotation {pitch=15, yaw=-5}
   }
   n8 = "Transform (Modify) Bone - Bone: base_jacket_mid_l"!(ComponentPose=@n7, bAlphaBoolEnabled:bool=True, Translation="3,5,5", Rotation="0,15,0", Scale="1,1,1") -> (Pose) {
    alphaInputType=Bool
    rotationMode=BMM_Additive
    rotationSpace=BCS_BoneSpace
    scaleMode=BMM_Additive
    scaleSpace=BCS_BoneSpace
    translationMode=BMM_Additive
    translationSpace=BCS_BoneSpace
    alphaBoolBlend {blendInTime=0.5, blendOption=Sinusoidal, blendOutTime=1}
    boneToModify {boneName=base_jacket_mid_l}
    rotation {yaw=15}
    translation=<3 5 5>
   }
   n9 = "Transform (Modify) Bone - Bone: base_jacket_back_l"!(ComponentPose=@n8, bAlphaBoolEnabled:bool=True, Translation="5,5,0", Rotation="-5,0,0", Scale="1,1,1") -> (Pose) {
    alphaInputType=Bool
    rotationMode=BMM_Additive
    rotationSpace=BCS_BoneSpace
    scaleMode=BMM_Additive
    scaleSpace=BCS_BoneSpace
    translationMode=BMM_Additive
    translationSpace=BCS_BoneSpace
    alphaBoolBlend {blendInTime=0.5, blendOption=Sinusoidal, blendOutTime=1}
    boneToModify {boneName=base_jacket_back_l}
    rotation {pitch=-5}
    translation=<5 5>
   }
   n10 = "Transform (Modify) Bone - Bone: base_jacket_back_r"!(ComponentPose=@n9, bAlphaBoolEnabled:bool=True, Translation="5,-5,0", Rotation="-5,0,0", Scale="1,1,1") -> (Pose) {
    alphaInputType=Bool
    rotationMode=BMM_Additive
    rotationSpace=BCS_BoneSpace
    scaleMode=BMM_Additive
    scaleSpace=BCS_BoneSpace
    translationMode=BMM_Additive
    translationSpace=BCS_BoneSpace
    alphaBoolBlend {blendInTime=0.5, blendOption=Sinusoidal, blendOutTime=1}
    boneToModify {boneName=base_jacket_back_r}
    rotation {pitch=-5}
    translation=<5 -5>
   }
   n11 = "Transform (Modify) Bone - Bone: base_jacket_mid_r"!(ComponentPose=@n10, bAlphaBoolEnabled:bool=True, Translation="3,-5,5", Rotation="0,-15,0", Scale="1,1,1") -> (Pose) {
    alphaInputType=Bool
    rotationMode=BMM_Additive
    rotationSpace=BCS_BoneSpace
    scaleMode=BMM_Additive
    scaleSpace=BCS_BoneSpace
    translationMode=BMM_Additive
    translationSpace=BCS_BoneSpace
    alphaBoolBlend {blendInTime=0.5, blendOption=Sinusoidal, blendOutTime=1}
    boneToModify {boneName=base_jacket_mid_r}
    rotation {yaw=-15}
    translation=<3 -5 5>
   }
   n12 = "Transform (Modify) Bone - Bone: base_necklace"!(ComponentPose=@n11, bAlphaBoolEnabled:bool=NOT Boolean(A=Get bIsTargeting()), Translation="0,0,0", Rotation="0,-5,0", Scale="1,1,1") -> (Pose) {
    alphaInputType=Bool
    rotationMode=BMM_Additive
    rotationSpace=BCS_BoneSpace
    scaleMode=BMM_Additive
    scaleSpace=BCS_BoneSpace
    translationMode=BMM_Additive
    translationSpace=BCS_BoneSpace
    alphaBoolBlend {blendInTime=0.5, blendOption=Sinusoidal, blendOutTime=1}
    boneToModify {boneName=base_necklace}
    rotation {yaw=-5}
   }
   n13 = "Transform (Modify) Bone - Bone: base_necklace"(ComponentPose=@n12, bAlphaBoolEnabled=True, Translation="-0.5,0,0", Rotation="0,-2.5,0", Scale="1,1,1") {
    alphaInputType=Bool
    rotationMode=BMM_Additive
    rotationSpace=BCS_BoneSpace
    scaleMode=BMM_Additive
    scaleSpace=BCS_BoneSpace
    translationMode=BMM_Additive
    translationSpace=BCS_BoneSpace
    alphaBoolBlend {blendInTime=0.5, blendOption=Sinusoidal, blendOutTime=1}
    boneToModify {boneName=base_necklace}
    rotation {yaw=-2.5}
    translation {x=-0.5}
   }
   n14 = "Transform (Modify) Bone - Bone: base_necklace"(ComponentPose=@n13, bAlphaBoolEnabled=True, Translation="0,0,0", Rotation="0,10,0", Scale="1,1,1") {
    alphaInputType=Bool
    rotationMode=BMM_Additive
    rotationSpace=BCS_BoneSpace
    scaleMode=BMM_Additive
    scaleSpace=BCS_BoneSpace
    translationMode=BMM_Additive
    translationSpace=BCS_BoneSpace
    alphaBoolBlend {blendInTime=0.5, blendOption=Sinusoidal, blendOutTime=0.5}
    boneToModify {boneName=base_necklace}
    rotation {yaw=10}
   }
   n15 = "Constraint - base_necklace_aimtarget (PPPP)"!(
    ComponentPose=@n14
    Alpha=1
    ConstraintWeights_0=0.8142
    ConstraintWeights_1:float=0.1402
    ConstraintWeights_2:float=0.0228
    ConstraintWeights_3:float=0.0228
   ) -> (Pose) {
    alphaBoolBlend {blendOption=Linear}
    boneToModify {boneName=base_necklace_aimtarget}
    constraintSetup [
     {targetBone {boneName=spine_04}}
     {targetBone {boneName=spine_05}}
     {targetBone {boneName=deform_pec_l}}
     {targetBone {boneName=deform_pec_r}}
    ]
    constraintWeights [1,1,1,1]
   }
   n16 = "Transform (Modify) Bone - Bone: base_necklace_aimtarget"!(ComponentPose=@n15, bAlphaBoolEnabled:bool=True, Translation="0,0,0.5", Rotation="0,0,0", Scale="1,1,1") -> (Pose) {
    alphaInputType=Bool
    rotationMode=BMM_Additive
    rotationSpace=BCS_BoneSpace
    scaleMode=BMM_Additive
    scaleSpace=BCS_BoneSpace
    translationMode=BMM_Additive
    translationSpace=BCS_BoneSpace
    alphaBoolBlend {blendInTime=0.5, blendOption=Sinusoidal, blendOutTime=0.5}
    boneToModify {boneName=base_necklace_aimtarget}
    translation {z=0.5}
   }
   n17 = Look At - Bone: dyn_necklace_1!(ComponentPose=@n16, bAlphaBoolEnabled:bool=True) -> (Pose) {
    alphaInputType=Bool
    bUseLookUpAxis=true
    alphaBoolBlend {blendOption=Linear}
    boneToModify {boneName=dyn_necklace_1}
    lookAtLocation {x=0}
    lookAtTarget {boneReference {boneName=base_necklace_aimtarget}}
    lookAt_Axis {axis=<1 0>}
    lookUp_Axis {axis {x=0, z=1}}
   }
   n18 = Look At - Bone: dyn_necklace_2!(ComponentPose=@n17, bAlphaBoolEnabled:bool=True) -> (Pose) {
    alphaInputType=Bool
    bUseLookUpAxis=true
    alphaBoolBlend {blendOption=Linear}
    boneToModify {boneName=dyn_necklace_2}
    lookAtLocation {x=0}
    lookAtTarget {boneReference {boneName=base_necklace_aimtarget}}
    lookAt_Axis {axis=<1 0>}
    lookUp_Axis {axis {x=0, z=1}}
   }
   Output Pose(Result=Blend Poses by bool(BlendPose_0=Use cached pose 'CACHE_ASWon'(), BlendPose_1=Component To Local(ComponentPose=@n18), bActiveValue=Get ASWon() -> (ASWon:bool)) {blendType=Linear}) {layerGroup=None}
   n19 = "Get Body Mesh "() -> ("Body Mesh ":Skeletal Mesh Component Object Reference)
   n20 = Layered blend per bone(
    BasePose=Layered blend per bone(
     BasePose=Copy Pose From Mesh() {bUseAttachedParent=true}
     BlendPoses_0=Copy Pose From Mesh(SourceMeshComponent=@n19) {bUseAttachedParent=true}
    )
    BlendPoses_0=Copy Pose From Mesh(SourceMeshComponent="Get Face Mesh "() -> ("Face Mesh ":Skeletal Mesh Component Object Reference)) {bUseAttachedParent=true}
   ) {layerSetup [{branchFilters [
       {boneName=FACIAL_C_FacialRoot}
       {boneName=FACIAL_C_Jaw}
       {boneName=FACIAL_C_Jawline}
       {boneName=FACIAL_R_Jawline}
       {boneName=FACIAL_L_Jawline}
       {boneName=FACIAL_R_UnderChin}
       {boneName=FACIAL_L_UnderChin}
       {boneName=FACIAL_C_UnderChin}
      ]}]}
   Save cached pose 'Start_Cache'(Pose=@n20) {cacheName=Start_Cache}
   n21 = "Transform (Modify) Bone - Bone: base_jacket_front_l"!(
    ComponentPose=Local To Component(LocalPose=Use cached pose 'CACHE_Start'())
    Alpha:float=1
    Translation="0,0,0"
    Rotation="10,0,0"
    Scale="1,1,1"
   ) -> (Pose) {
    rotationMode=BMM_Additive
    rotationSpace=BCS_BoneSpace
    scaleMode=BMM_Additive
    scaleSpace=BCS_BoneSpace
    translationMode=BMM_Additive
    translationSpace=BCS_BoneSpace
    alphaBoolBlend {blendOption=Linear}
    boneToModify {boneName=base_jacket_front_l}
    rotation {pitch=10}
   }
   n22 = "Transform (Modify) Bone - Bone: base_jacket_front_r"!(ComponentPose=@n21, Alpha:float=1, Translation="0,0,0", Rotation="10,0,0", Scale="1,1,1") -> (Pose) {
    rotationMode=BMM_Additive
    rotationSpace=BCS_BoneSpace
    scaleMode=BMM_Additive
    scaleSpace=BCS_BoneSpace
    translationMode=BMM_Additive
    translationSpace=BCS_BoneSpace
    alphaBoolBlend {blendOption=Linear}
    boneToModify {boneName=base_jacket_front_r}
    rotation {pitch=10}
   }
   n23 = "Transform (Modify) Bone - Bone: base_jacket_front_l"!(ComponentPose=@n22, bAlphaBoolEnabled:bool=True, Translation="0,0,0", Rotation="-15,0,0", Scale="1,1,1") -> (Pose) {
    alphaInputType=Bool
    rotationMode=BMM_Additive
    rotationSpace=BCS_BoneSpace
    scaleMode=BMM_Additive
    scaleSpace=BCS_BoneSpace
    translationMode=BMM_Additive
    translationSpace=BCS_BoneSpace
    alphaBoolBlend {blendInTime=0.5, blendOption=QuadraticInOut, blendOutTime=1}
    boneToModify {boneName=base_jacket_front_l}
    rotation {pitch=-15}
   }
   n24 = "Transform (Modify) Bone - Bone: base_jacket_front_r"!(ComponentPose=@n23, bAlphaBoolEnabled:bool=True, Translation="0,0,0", Rotation="-15,0,0", Scale="1,1,1") -> (Pose) {
    alphaInputType=Bool
    rotationMode=BMM_Additive
    rotationSpace=BCS_BoneSpace
    scaleMode=BMM_Additive
    scaleSpace=BCS_BoneSpace
    translationMode=BMM_Additive
    translationSpace=BCS_BoneSpace
    alphaBoolBlend {blendInTime=0.5, blendOption=QuadraticInOut, blendOutTime=1}
    boneToModify {boneName=base_jacket_front_r}
    rotation {pitch=-15}
   }
   Save cached pose 'CACHE_ASWon'(Pose=Component To Local(ComponentPose=@n24)) {cacheName=CACHE_ASWon}
   n25 = "Constraint - base_neck_front_r (PPP)"!(
    ComponentPose=Local To Component(LocalPose=Use cached pose 'Start_Cache'())
    Alpha=1
    ConstraintWeights_0=0.5
    ConstraintWeights_1:float=0.25
    ConstraintWeights_2:float=0.25
   ) -> (Pose) {
    alphaBoolBlend {blendOption=Linear}
    boneToModify {boneName=base_neck_front_r}
    constraintSetup [{targetBone {boneName=clavicle_r}}, {targetBone {boneName=spine_05}}, {targetBone {boneName=neck_02}}]
    constraintWeights [1,1,1]
   }
   n26 = "Constraint - base_neck_mid_r (PPP)"!(ComponentPose=@n25, Alpha=1, ConstraintWeights_0=0.4, ConstraintWeights_1:float=0.3, ConstraintWeights_2:float=0.3) -> (Pose) {
    alphaBoolBlend {blendOption=Linear}
    boneToModify {boneName=base_neck_mid_r}
    constraintSetup [{targetBone {boneName=clavicle_r}}, {targetBone {boneName=spine_05}}, {targetBone {boneName=neck_02}}]
    constraintWeights [1,1,1]
   }
   n27 = "Constraint - base_neck_midback_r (PPP)"!(ComponentPose=@n26, Alpha=1, ConstraintWeights_0=0.3, ConstraintWeights_1:float=0.3, ConstraintWeights_2:float=0.4) -> (Pose) {
    alphaBoolBlend {blendOption=Linear}
    boneToModify {boneName=base_neck_midback_r}
    constraintSetup [{targetBone {boneName=clavicle_r}}, {targetBone {boneName=spine_05}}, {targetBone {boneName=neck_02}}]
    constraintWeights [1,1,1]
   }
   n28 = "Constraint - base_neck_back (PP)"!(ComponentPose=@n27, Alpha=1, ConstraintWeights_0=0.7, ConstraintWeights_1:float=0.3) -> (Pose) {
    alphaBoolBlend {blendOption=Linear}
    boneToModify {boneName=base_neck_back}
    constraintSetup [{targetBone {boneName=spine_05}}, {targetBone {boneName=neck_02}}]
    constraintWeights [1,1]
   }
   n29 = "Constraint - base_neck_mid_l (PPP)"!(ComponentPose=@n28, Alpha=1, ConstraintWeights_0=0.4, ConstraintWeights_1:float=0.3, ConstraintWeights_2:float=0.3) -> (Pose) {
    alphaBoolBlend {blendOption=Linear}
    boneToModify {boneName=base_neck_mid_l}
    constraintSetup [{targetBone {boneName=clavicle_l}}, {targetBone {boneName=spine_05}}, {targetBone {boneName=neck_02}}]
    constraintWeights [1,1,1]
   }
   n30 = "Constraint - base_neck_front_l (PPP)"!(ComponentPose=@n29, Alpha=1, ConstraintWeights_0=0.5, ConstraintWeights_1:float=0.25, ConstraintWeights_2:float=0.25) -> (Pose) {
    alphaBoolBlend {blendOption=Linear}
    boneToModify {boneName=base_neck_front_l}
    constraintSetup [{targetBone {boneName=clavicle_l}}, {targetBone {boneName=spine_05}}, {targetBone {boneName=neck_02}}]
    constraintWeights [1,1,1]
   }
   n31 = "Constraint - base_necklace (PPP)"!(ComponentPose=@n30, Alpha=1, ConstraintWeights_0=0.9661, ConstraintWeights_1:float=0.0326, ConstraintWeights_2:float=0.0013) -> (Pose) {
    alphaBoolBlend {blendOption=Linear}
    boneToModify {boneName=base_necklace}
    constraintSetup [{targetBone {boneName=spine_05}}, {targetBone {boneName=neck_01}}, {targetBone {boneName=clavicle_r}}]
    constraintWeights [1,1,1]
   }
   n32 = "Constraint - base_frontfurball_up_r (PPPPPPPP)"!(
    ComponentPose=@n31
    Alpha=1
    ConstraintWeights_0=0.3667
    ConstraintWeights_1:float=0.3348
    ConstraintWeights_2:float=0.1431
    ConstraintWeights_3:float=0.0763
    ConstraintWeights_4:float=0.043
    ConstraintWeights_5:float=0.0284
    ConstraintWeights_6:float=0.006
    ConstraintWeights_7:float=0.0018
   ) -> (Pose) {
    alphaBoolBlend {blendOption=Linear}
    boneToModify {boneName=base_frontfurball_up_r}
    constraintSetup [
     {targetBone {boneName=spine_04}}
     {targetBone {boneName=deform_pec_r}}
     {targetBone {boneName=spine_05}}
     {targetBone {boneName=clavicle_r}}
     {targetBone {boneName=upperarm_twist_01_r}}
     {targetBone {boneName=deform_upperarm_fr_r}}
     {targetBone {boneName=spine_02}}
     {targetBone {boneName=spine_03}}
    ]
    constraintWeights [1,1,1,1,1,1,1,1]
   }
   n33 = "Constraint - base_frontfurball_up_l (PPPPPPPP)"!(
    ComponentPose=@n32
    Alpha=1
    ConstraintWeights_0=0.3115
    ConstraintWeights_1:float=0.2476
    ConstraintWeights_2:float=0.1429
    ConstraintWeights_3:float=0.1376
    ConstraintWeights_4:float=0.0871
    ConstraintWeights_5:float=0.069
    ConstraintWeights_6:float=0.0022
    ConstraintWeights_7:float=0.0021
   ) -> (Pose) {
    alphaBoolBlend {blendOption=Linear}
    boneToModify {boneName=base_frontfurball_up_l}
    constraintSetup [
     {targetBone {boneName=spine_04}}
     {targetBone {boneName=deform_pec_l}}
     {targetBone {boneName=spine_05}}
     {targetBone {boneName=clavicle_l}}
     {targetBone {boneName=upperarm_twist_01_l}}
     {targetBone {boneName=deform_upperarm_fr_l}}
     {targetBone {boneName=spine_02}}
     {targetBone {boneName=spine_03}}
    ]
    constraintWeights [1,1,1,1,1,1,1,1]
   }
   n34 = "Constraint - base_frontfurball_down_r (PPPPPPPP)"!(
    ComponentPose=@n33
    Alpha=1
    ConstraintWeights_0=0.6018
    ConstraintWeights_1:float=0.2477
    ConstraintWeights_2:float=0.0471
    ConstraintWeights_3:float=0.0425
    ConstraintWeights_4:float=0.0375
    ConstraintWeights_5:float=0.0174
    ConstraintWeights_6:float=0.0032
    ConstraintWeights_7:float=0.0029
   ) -> (Pose) {
    alphaBoolBlend {blendOption=Linear}
    boneToModify {boneName=base_frontfurball_down_r}
    constraintSetup [
     {targetBone {boneName=spine_04}}
     {targetBone {boneName=deform_pec_r}}
     {targetBone {boneName=spine_05}}
     {targetBone {boneName=spine_02}}
     {targetBone {boneName=spine_03}}
     {targetBone {boneName=dyn_jacket_front_r_1}}
     {targetBone {boneName=clavicle_r}}
     {targetBone {boneName=upperarm_twist_01_r}}
    ]
    constraintWeights [1,1,1,1,1,1,1,1]
   }
   n35 = "Constraint - base_frontfurball_down_l (PPPPPPPP)"!(
    ComponentPose=@n34
    Alpha=1
    ConstraintWeights_0=0.6018
    ConstraintWeights_1:float=0.2477
    ConstraintWeights_2:float=0.0471
    ConstraintWeights_3:float=0.0425
    ConstraintWeights_4:float=0.0375
    ConstraintWeights_5:float=0.0174
    ConstraintWeights_6:float=0.0032
    ConstraintWeights_7:float=0.0029
   ) -> (Pose) {
    alphaBoolBlend {blendOption=Linear}
    boneToModify {boneName=base_frontfurball_down_l}
    constraintSetup [
     {targetBone {boneName=spine_04}}
     {targetBone {boneName=deform_pec_l}}
     {targetBone {boneName=spine_05}}
     {targetBone {boneName=spine_02}}
     {targetBone {boneName=spine_03}}
     {targetBone {boneName=dyn_jacket_front_l_1}}
     {targetBone {boneName=clavicle_l}}
     {targetBone {boneName=upperarm_twist_01_l}}
    ]
    constraintWeights [1,1,1,1,1,1,1,1]
   }
   n36 = Anim Dynamics - Bone: dyn_frontfurball_up_r(ComponentPose=@n35, Alpha=1, ExternalForce="0,0,0") {
    angularDampingOverride=1
    angularSpringConstant=20000
    bAngularSpring=true
    bOverrideAngularDamping=true
    bUseGravityOverride=true
    gravityScale=0.35
    simulationSpace=RootRelative
    alphaBoolBlend {blendOption=Linear}
    boundBone {boneName=dyn_frontfurball_up_r}
    componentLinearAccScale=<1 1 1>
    componentLinearVelScale=<1 1 1>
    gravityOverride {z=-980}
    physicsBodyDefinitions [{
      boundBone {boneName=dyn_frontfurball_up_r}
      boxExtents=<20 30 30>
      constraintSetup {angularLimitsMax=<5 10 5>, angularLimitsMin=<-5 -10 -10>, angularTarget=<1 0 0>}
      localJointOffset=<4 0 0>
     }]
   }
   n37 = Anim Dynamics - Bone: dyn_frontfurball_up_l(ComponentPose=@n36, Alpha=1, ExternalForce="0,0,0") {
    angularDampingOverride=1
    angularSpringConstant=20000
    bAngularSpring=true
    bOverrideAngularDamping=true
    bUseGravityOverride=true
    gravityScale=0.35
    simulationSpace=RootRelative
    alphaBoolBlend {blendOption=Linear}
    boundBone {boneName=dyn_frontfurball_up_l}
    componentLinearAccScale=<1 1 1>
    componentLinearVelScale=<1 1 1>
    gravityOverride {z=-980}
    physicsBodyDefinitions [{
      boundBone {boneName=dyn_frontfurball_up_l}
      boxExtents=<20 30 30>
      constraintSetup {angularLimitsMax=<5 10 5>, angularLimitsMin=<-5 -10 -10>, angularTarget=<1 0 0>}
      localJointOffset=<4 0 0>
     }]
   }
   n38 = Anim Dynamics - Bone: dyn_frontfurball_down_r(ComponentPose=@n37, Alpha=1, ExternalForce="0,0,0") {
    angularDampingOverride=1
    angularSpringConstant=20000
    bAngularSpring=true
    bOverrideAngularDamping=true
    bUseGravityOverride=true
    gravityScale=0.35
    simulationSpace=RootRelative
    alphaBoolBlend {blendOption=Linear}
    boundBone {boneName=dyn_frontfurball_down_r}
    componentLinearAccScale=<1 1 1>
    componentLinearVelScale=<1 1 1>
    gravityOverride {z=-980}
    physicsBodyDefinitions [{
      boundBone {boneName=dyn_frontfurball_down_r}
      boxExtents=<20 30 30>
      constraintSetup {angularLimitsMax=<5 10 10>, angularLimitsMin=<-5 -10 -10>, angularTarget=<1 0 0>}
      localJointOffset=<4 0 0>
     }]
   }
   n39 = Anim Dynamics - Bone: dyn_frontfurball_up_l(ComponentPose=@n38, Alpha=1, ExternalForce="0,0,0") {
    angularDampingOverride=1
    angularSpringConstant=20000
    bAngularSpring=true
    bOverrideAngularDamping=true
    bUseGravityOverride=true
    gravityScale=0.35
    simulationSpace=RootRelative
    alphaBoolBlend {blendOption=Linear}
    boundBone {boneName=dyn_frontfurball_up_l}
    componentLinearAccScale=<1 1 1>
    componentLinearVelScale=<1 1 1>
    gravityOverride {z=-980}
    physicsBodyDefinitions [{
      boundBone {boneName=dyn_frontfurball_up_l}
      boxExtents=<20 30 30>
      constraintSetup {angularLimitsMax=<5 10 10>, angularLimitsMin=<-5 -10 -10>, angularTarget=<1 0 0>}
      localJointOffset=<4 0 0>
     }]
   }
   n40 = Control Rig!(
    Source=Component To Local(ComponentPose=@n39)
    Alpha
    IsCrouching:bool=Get bIsCrouching()
    Moving2D:bool=Get bIsMoving2D()
    IsTargeting:bool=Get bIsTargeting()
    SkyFloating:bool=Get bIsSkydiveFloating()
   ) -> (Pose) {
    alphaBoolBlend {blendOption=Linear}
    controlRigAssetReference {blueprintRigClass=&/F_MED_ForestBath_CtrlRig.F_MED_ForestBath_CtrlRig_C}
   }
   n41 = Anim Dynamics - Bone: dyn_lowerpants_l(ComponentPose=Local To Component(LocalPose=@n40), Alpha=1, ExternalForce="0,0,0") {
    angularDampingOverride=1
    angularSpringConstant=20000
    bAngularSpring=true
    bOverrideAngularDamping=true
    bUseGravityOverride=true
    gravityScale=0.35
    simulationSpace=RootRelative
    alphaBoolBlend {blendOption=Linear}
    boundBone {boneName=dyn_lowerpants_l}
    componentLinearAccScale=<1 1 1>
    componentLinearVelScale=<1 1 1>
    gravityOverride {z=-980}
    physicsBodyDefinitions [{
      boundBone {boneName=dyn_lowerpants_l}
      boxExtents=<20 30 30>
      constraintSetup {angularLimitsMax=<30 30 5>, angularLimitsMin=<-30 -30 -30>, angularTarget=<1 0 0>}
      localJointOffset=<4 0 0>
     }]
    relativeSpaceBone {boneName=root}
   }
   n42 = Anim Dynamics - Bone: dyn_lowerpants_r(ComponentPose=@n41, Alpha=1, ExternalForce="0,0,0") {
    angularDampingOverride=1
    angularSpringConstant=20000
    bAngularSpring=true
    bOverrideAngularDamping=true
    bUseGravityOverride=true
    gravityScale=0.35
    simulationSpace=RootRelative
    alphaBoolBlend {blendOption=Linear}
    boundBone {boneName=dyn_lowerpants_r}
    componentLinearAccScale=<1 1 1>
    componentLinearVelScale=<1 1 1>
    gravityOverride {z=-980}
    physicsBodyDefinitions [{
      boundBone {boneName=dyn_lowerpants_r}
      boxExtents=<20 30 30>
      constraintSetup {angularLimitsMax=<30 30 5>, angularLimitsMin=<-30 -30 -30>, angularTarget=<1 0 0>}
      localJointOffset=<4 0 0>
     }]
   }
   n43 = "Transform (Modify) Bone - Bone: base_necklace"!(ComponentPose=@n42, Alpha:float=1, Translation="-1,1,0", Rotation="0,0,0", Scale="1,1,1") -> (Pose) {
    rotationMode=BMM_Additive
    rotationSpace=BCS_BoneSpace
    scaleMode=BMM_Additive
    scaleSpace=BCS_BoneSpace
    translationMode=BMM_Additive
    translationSpace=BCS_BoneSpace
    alphaBoolBlend {blendInTime=0.5, blendOption=Sinusoidal, blendOutTime=1}
    boneToModify {boneName=base_necklace}
    translation=<-1 1>
   }
   n44 = "Transform (Modify) Bone - Bone: base_jacket_mid_l"(ComponentPose=@n43, bAlphaBoolEnabled=True, Translation="0,0,2", Rotation="10,-5,0", Scale="1,1,1") {
    alphaInputType=Bool
    rotationMode=BMM_Additive
    rotationSpace=BCS_BoneSpace
    scaleMode=BMM_Additive
    scaleSpace=BCS_BoneSpace
    translationMode=BMM_Additive
    translationSpace=BCS_BoneSpace
    alphaBoolBlend {blendInTime=0.5, blendOption=Sinusoidal, blendOutTime=1}
    boneToModify {boneName=base_jacket_mid_l}
    rotation {pitch=10, yaw=-5}
    translation {z=2}
   }
   n45 = "Transform (Modify) Bone - Bone: base_jacket_back_l"(ComponentPose=@n44, bAlphaBoolEnabled=True, Translation="0,0,2", Rotation="10,5,0", Scale="1,1,1") {
    alphaInputType=Bool
    rotationMode=BMM_Additive
    rotationSpace=BCS_BoneSpace
    scaleMode=BMM_Additive
    scaleSpace=BCS_BoneSpace
    translationMode=BMM_Additive
    translationSpace=BCS_BoneSpace
    alphaBoolBlend {blendInTime=0.5, blendOption=Sinusoidal, blendOutTime=1}
    boneToModify {boneName=base_jacket_back_l}
    rotation {pitch=10, yaw=5}
    translation {z=2}
   }
   n46 = "Transform (Modify) Bone - Bone: base_jacket_back_r"(ComponentPose=@n45, bAlphaBoolEnabled=True, Translation="0,0,2", Rotation="10,-5,0", Scale="1,1,1") {
    alphaInputType=Bool
    rotationMode=BMM_Additive
    rotationSpace=BCS_BoneSpace
    scaleMode=BMM_Additive
    scaleSpace=BCS_BoneSpace
    translationMode=BMM_Additive
    translationSpace=BCS_BoneSpace
    alphaBoolBlend {blendInTime=0.5, blendOption=Sinusoidal, blendOutTime=1}
    boneToModify {boneName=base_jacket_back_r}
    rotation {pitch=10, yaw=-5}
    translation {z=2}
   }
   n47 = "Transform (Modify) Bone - Bone: base_jacket_mid_r"(ComponentPose=@n46, bAlphaBoolEnabled=True, Translation="0,0,2", Rotation="10,5,0", Scale="1,1,1") {
    alphaInputType=Bool
    rotationMode=BMM_Additive
    rotationSpace=BCS_BoneSpace
    scaleMode=BMM_Additive
    scaleSpace=BCS_BoneSpace
    translationMode=BMM_Additive
    translationSpace=BCS_BoneSpace
    alphaBoolBlend {blendInTime=0.5, blendOption=Sinusoidal, blendOutTime=1}
    boneToModify {boneName=base_jacket_mid_r}
    rotation {pitch=10, yaw=5}
    translation {z=2}
   }
   n48 = Get ParachutBeginning() -> (ParachutBeginning:bool)
   n49 = "Transform (Modify) Bone - Bone: base_jacket_back_l"(ComponentPose=@n47, bAlphaBoolEnabled=@n48, Translation="0,0,0", Rotation="45,0,0", Scale="1,1,1") {
    alphaInputType=Bool
    rotationMode=BMM_Additive
    rotationSpace=BCS_BoneSpace
    scaleMode=BMM_Additive
    scaleSpace=BCS_BoneSpace
    translationMode=BMM_Additive
    translationSpace=BCS_BoneSpace
    alphaBoolBlend {blendOption=Sinusoidal, blendOutTime=1}
    boneToModify {boneName=base_jacket_back_l}
    rotation {pitch=45}
   }
   n50 = "Transform (Modify) Bone - Bone: base_jacket_back_r"(ComponentPose=@n49, bAlphaBoolEnabled=@n48, Translation="0,0,0", Rotation="45,0,0", Scale="1,1,1") {
    alphaInputType=Bool
    rotationMode=BMM_Additive
    rotationSpace=BCS_BoneSpace
    scaleMode=BMM_Additive
    scaleSpace=BCS_BoneSpace
    translationMode=BMM_Additive
    translationSpace=BCS_BoneSpace
    alphaBoolBlend {blendOption=Sinusoidal, blendOutTime=1}
    boneToModify {boneName=base_jacket_back_r}
    rotation {pitch=45}
   }
   n51 = Get ParachutMoving() -> (ParachutMoving:bool)
   n52 = "Transform (Modify) Bone - Bone: base_jacket_back_l"(ComponentPose=@n50, bAlphaBoolEnabled=@n51, Translation="0,0,0", Rotation="10,0,0", Scale="1,1,1") {
    alphaInputType=Bool
    rotationMode=BMM_Additive
    rotationSpace=BCS_BoneSpace
    scaleMode=BMM_Additive
    scaleSpace=BCS_BoneSpace
    translationMode=BMM_Additive
    translationSpace=BCS_BoneSpace
    alphaBoolBlend {blendOption=Sinusoidal, blendOutTime=1}
    boneToModify {boneName=base_jacket_back_l}
    rotation {pitch=10}
   }
   n53 = "Transform (Modify) Bone - Bone: base_jacket_back_r"(ComponentPose=@n52, bAlphaBoolEnabled=@n51, Translation="0,0,0", Rotation="10,0,0", Scale="1,1,1") {
    alphaInputType=Bool
    rotationMode=BMM_Additive
    rotationSpace=BCS_BoneSpace
    scaleMode=BMM_Additive
    scaleSpace=BCS_BoneSpace
    translationMode=BMM_Additive
    translationSpace=BCS_BoneSpace
    alphaBoolBlend {blendOption=Sinusoidal, blendOutTime=1}
    boneToModify {boneName=base_jacket_back_r}
    rotation {pitch=10}
   }
   n54 = "Transform (Modify) Bone - Bone: base_necklace"!(ComponentPose=@n53, bAlphaBoolEnabled:bool=True, Translation="2,0,0", Rotation="29,0,0", Scale="1,1,1") -> (Pose) {
    alphaInputType=Bool
    rotationMode=BMM_Additive
    rotationSpace=BCS_BoneSpace
    scaleMode=BMM_Additive
    scaleSpace=BCS_BoneSpace
    translationMode=BMM_Additive
    translationSpace=BCS_BoneSpace
    alphaBoolBlend {blendInTime=0.5, blendOption=Sinusoidal, blendOutTime=1}
    boneToModify {boneName=base_necklace}
    rotation {pitch=29}
    translation {x=2}
   }
   n55 = "Transform (Modify) Bone - Bone: base_jacket_front_l"(ComponentPose=@n54, bAlphaBoolEnabled=True, Translation="0,0,0", Rotation="15,0,10", Scale="1,1,1") {
    alphaInputType=Bool
    rotationMode=BMM_Additive
    rotationSpace=BCS_BoneSpace
    scaleMode=BMM_Additive
    scaleSpace=BCS_BoneSpace
    translationMode=BMM_Additive
    translationSpace=BCS_BoneSpace
    alphaBoolBlend {blendInTime=0.5, blendOption=Sinusoidal, blendOutTime=1}
    boneToModify {boneName=base_jacket_front_l}
    rotation {pitch=15, roll=10}
   }
   n56 = "Transform (Modify) Bone - Bone: base_jacket_front_r"(ComponentPose=@n55, bAlphaBoolEnabled=True, Translation="0,0,0", Rotation="15,0,10", Scale="1,1,1") {
    alphaInputType=Bool
    rotationMode=BMM_Additive
    rotationSpace=BCS_BoneSpace
    scaleMode=BMM_Additive
    scaleSpace=BCS_BoneSpace
    translationMode=BMM_Additive
    translationSpace=BCS_BoneSpace
    alphaBoolBlend {blendInTime=0.5, blendOption=Sinusoidal, blendOutTime=1}
    boneToModify {boneName=base_jacket_front_r}
    rotation {pitch=15, roll=10}
   }
   n57 = "Transform (Modify) Bone - Bone: base_jacket_front_l"(ComponentPose=@n56, bAlphaBoolEnabled=True, Translation="0,0,3", Rotation="10,0,20", Scale="1,1,1") {
    alphaInputType=Bool
    rotationMode=BMM_Additive
    rotationSpace=BCS_BoneSpace
    scaleMode=BMM_Additive
    scaleSpace=BCS_BoneSpace
    translationMode=BMM_Additive
    translationSpace=BCS_BoneSpace
    alphaBoolBlend {blendInTime=0.5, blendOption=Sinusoidal, blendOutTime=1}
    boneToModify {boneName=base_jacket_front_l}
    rotation {pitch=10, roll=20}
    translation {z=3}
   }
   n58 = "Transform (Modify) Bone - Bone: base_jacket_front_r"(ComponentPose=@n57, bAlphaBoolEnabled=True, Translation="0,0,3", Rotation="10,0,20", Scale="1,1,1") {
    alphaInputType=Bool
    rotationMode=BMM_Additive
    rotationSpace=BCS_BoneSpace
    scaleMode=BMM_Additive
    scaleSpace=BCS_BoneSpace
    translationMode=BMM_Additive
    translationSpace=BCS_BoneSpace
    alphaBoolBlend {blendInTime=0.5, blendOption=Sinusoidal, blendOutTime=1}
    boneToModify {boneName=base_jacket_front_r}
    rotation {pitch=10, roll=20}
    translation {z=3}
   }
   n59 = "Transform (Modify) Bone - Bone: base_jacket_front_l"(ComponentPose=@n58, bAlphaBoolEnabled=True, Translation="0,0,3", Rotation="3,0,-45", Scale="1,1,1") {
    alphaInputType=Bool
    rotationMode=BMM_Additive
    rotationSpace=BCS_BoneSpace
    scaleMode=BMM_Additive
    scaleSpace=BCS_BoneSpace
    translationMode=BMM_Additive
    translationSpace=BCS_BoneSpace
    alphaBoolBlend {blendInTime=0.5, blendOption=Sinusoidal, blendOutTime=1}
    boneToModify {boneName=base_jacket_front_l}
    rotation {pitch=3, roll=-45}
    translation {z=3}
   }
   n60 = "Transform (Modify) Bone - Bone: base_jacket_front_r"(ComponentPose=@n59, bAlphaBoolEnabled=True, Translation="0,0,3", Rotation="0,0,-45", Scale="1,1,1") {
    alphaInputType=Bool
    rotationMode=BMM_Additive
    rotationSpace=BCS_BoneSpace
    scaleMode=BMM_Additive
    scaleSpace=BCS_BoneSpace
    translationMode=BMM_Additive
    translationSpace=BCS_BoneSpace
    alphaBoolBlend {blendInTime=0.5, blendOption=Sinusoidal, blendOutTime=1}
    boneToModify {boneName=base_jacket_front_r}
    rotation {roll=-45}
    translation {z=3}
   }
   n61 = "Transform (Modify) Bone - Bone: base_jacket_front_l"(ComponentPose=@n60, bAlphaBoolEnabled=True, Translation="0,0,0", Rotation="11,0,10", Scale="1,1,1") {
    alphaInputType=Bool
    rotationMode=BMM_Additive
    rotationSpace=BCS_BoneSpace
    scaleMode=BMM_Additive
    scaleSpace=BCS_BoneSpace
    translationMode=BMM_Additive
    translationSpace=BCS_BoneSpace
    alphaBoolBlend {blendInTime=0.5, blendOption=Sinusoidal, blendOutTime=1}
    boneToModify {boneName=base_jacket_front_l}
    rotation {pitch=11, roll=10}
   }
   n62 = "Transform (Modify) Bone - Bone: base_jacket_front_r"(ComponentPose=@n61, bAlphaBoolEnabled=True, Translation="0,0,0", Rotation="10,0,10", Scale="1,1,1") {
    alphaInputType=Bool
    rotationMode=BMM_Additive
    rotationSpace=BCS_BoneSpace
    scaleMode=BMM_Additive
    scaleSpace=BCS_BoneSpace
    translationMode=BMM_Additive
    translationSpace=BCS_BoneSpace
    alphaBoolBlend {blendInTime=0.5, blendOption=Sinusoidal, blendOutTime=1}
    boneToModify {boneName=base_jacket_front_r}
    rotation {pitch=10, roll=10}
   }
   n63 = "Transform (Modify) Bone - Bone: base_jacket_front_r"(ComponentPose=@n62, bAlphaBoolEnabled=True, Translation="5,0,0", Rotation="60,60,60", Scale="1,1,1") {
    alphaInputType=Bool
    rotationMode=BMM_Additive
    rotationSpace=BCS_BoneSpace
    scaleMode=BMM_Additive
    scaleSpace=BCS_BoneSpace
    translationMode=BMM_Additive
    translationSpace=BCS_BoneSpace
    alphaBoolBlend {blendInTime=2, blendOption=Sinusoidal}
    boneToModify {boneName=base_jacket_front_r}
    rotation=<rot 60 60 60>
    translation {x=5}
   }
   n64 = "Transform (Modify) Bone - Bone: base_jacket_mid_r"(ComponentPose=@n63, bAlphaBoolEnabled=True, Translation="0,0,0", Rotation="40,0,30", Scale="1,1,1") {
    alphaInputType=Bool
    rotationMode=BMM_Additive
    rotationSpace=BCS_BoneSpace
    scaleMode=BMM_Additive
    scaleSpace=BCS_BoneSpace
    translationMode=BMM_Additive
    translationSpace=BCS_BoneSpace
    alphaBoolBlend {blendInTime=2, blendOption=Sinusoidal}
    boneToModify {boneName=base_jacket_mid_r}
    rotation {pitch=40, roll=30}
   }
   n65 = "Transform (Modify) Bone - Bone: base_jacket_back_r"(ComponentPose=@n64, bAlphaBoolEnabled=True, Translation="0,0,0", Rotation="30,-30,30", Scale="1,1,1") {
    alphaInputType=Bool
    rotationMode=BMM_Additive
    rotationSpace=BCS_BoneSpace
    scaleMode=BMM_Additive
    scaleSpace=BCS_BoneSpace
    translationMode=BMM_Additive
    translationSpace=BCS_BoneSpace
    alphaBoolBlend {blendInTime=2, blendOption=Sinusoidal}
    boneToModify {boneName=base_jacket_back_r}
    rotation=<rot 30 -30 30>
   }
   n66 = "Transform (Modify) Bone - Bone: base_jacket_front_l"(ComponentPose=@n65, bAlphaBoolEnabled=True, Translation="0,0,1", Rotation="5,0,-12", Scale="1,1,1") {
    alphaInputType=Bool
    rotationMode=BMM_Additive
    rotationSpace=BCS_BoneSpace
    scaleMode=BMM_Additive
    scaleSpace=BCS_BoneSpace
    translationMode=BMM_Additive
    translationSpace=BCS_BoneSpace
    alphaBoolBlend {blendInTime=0.5, blendOption=Sinusoidal, blendOutTime=1}
    boneToModify {boneName=base_jacket_front_l}
    rotation {pitch=5, roll=-12}
    translation {z=1}
   }
   n67 = "Transform (Modify) Bone - Bone: base_jacket_front_r"(ComponentPose=@n66, bAlphaBoolEnabled=True, Translation="0,0,1", Rotation="5,0,12", Scale="1,1,1") {
    alphaInputType=Bool
    rotationMode=BMM_Additive
    rotationSpace=BCS_BoneSpace
    scaleMode=BMM_Additive
    scaleSpace=BCS_BoneSpace
    translationMode=BMM_Additive
    translationSpace=BCS_BoneSpace
    alphaBoolBlend {blendInTime=0.5, blendOption=Sinusoidal, blendOutTime=1}
    boneToModify {boneName=base_jacket_front_r}
    rotation {pitch=5, roll=12}
    translation {z=1}
   }
   n68 = "Transform (Modify) Bone - Bone: base_jacket_front_r"(ComponentPose=@n67, bAlphaBoolEnabled=True, Translation="0,0,5", Rotation="10,0,0", Scale="1,1,1") {
    alphaInputType=Bool
    rotationMode=BMM_Additive
    rotationSpace=BCS_BoneSpace
    scaleMode=BMM_Additive
    scaleSpace=BCS_BoneSpace
    translationMode=BMM_Additive
    translationSpace=BCS_BoneSpace
    alphaBoolBlend {blendInTime=0.5, blendOption=Sinusoidal, blendOutTime=1}
    boneToModify {boneName=base_jacket_front_r}
    rotation {pitch=10}
    translation {z=5}
   }
   n69 = "Transform (Modify) Bone - Bone: base_jacket_mid_r"(ComponentPose=@n68, bAlphaBoolEnabled=True, Translation="0,0,0", Rotation="5,0,0", Scale="1,1,1") {
    alphaInputType=Bool
    rotationMode=BMM_Additive
    rotationSpace=BCS_BoneSpace
    scaleMode=BMM_Additive
    scaleSpace=BCS_BoneSpace
    translationMode=BMM_Additive
    translationSpace=BCS_BoneSpace
    alphaBoolBlend {blendInTime=0.5, blendOption=Sinusoidal, blendOutTime=1}
    boneToModify {boneName=base_jacket_mid_r}
    rotation {pitch=5}
   }
   n70 = "Transform (Modify) Bone - Bone: attach_backpack"!(ComponentPose=@n69, bAlphaBoolEnabled:bool=True, Translation="0,5,0", Rotation="0,0,8", Scale="1,1,1") -> (Pose) {
    alphaInputType=Bool
    rotationMode=BMM_Additive
    rotationSpace=BCS_BoneSpace
    scaleMode=BMM_Additive
    scaleSpace=BCS_BoneSpace
    translationMode=BMM_Additive
    translationSpace=BCS_BoneSpace
    alphaBoolBlend {blendInTime=0.5, blendOption=Sinusoidal, blendOutTime=1}
    boneToModify {boneName=attach_backpack}
    rotation {roll=8}
    translation {y=5}
   }
   n71 = "Transform (Modify) Bone - Bone: attach_cape"!(ComponentPose=@n70, bAlphaBoolEnabled:bool=True, Translation="0,1,0", Rotation="0,0,0", Scale="1,1,1") -> (Pose) {
    alphaInputType=Bool
    rotationMode=BMM_Additive
    rotationSpace=BCS_BoneSpace
    scaleMode=BMM_Additive
    scaleSpace=BCS_BoneSpace
    translationMode=BMM_Additive
    translationSpace=BCS_BoneSpace
    alphaBoolBlend {blendInTime=0.5, blendOption=Sinusoidal, blendOutTime=1}
    boneToModify {boneName=attach_cape}
    translation {y=1}
   }
   n72 = "Transform (Modify) Bone - Bone: neck_02"!(ComponentPose=@n71, bAlphaBoolEnabled:bool=True, Translation="0,0,0", Rotation="0,0,0", Scale="1,1.04,1.04") -> (Pose) {
    alphaInputType=Bool
    scaleMode=BMM_Additive
    scaleSpace=BCS_BoneSpace
    alphaBoolBlend {blendInTime=0.5, blendOption=Sinusoidal, blendOutTime=1}
    boneToModify {boneName=neck_02}
    scale {y=1.04, z=1.04}
   }
   n73 = "Transform (Modify) Bone - Bone: head"!(ComponentPose=@n72, bAlphaBoolEnabled:bool=True, Translation="0,0,0", Rotation="0,0,0", Scale="1,0.96,0.96") -> (Pose) {
    alphaInputType=Bool
    scaleMode=BMM_Additive
    scaleSpace=BCS_BoneSpace
    alphaBoolBlend {blendInTime=0.5, blendOption=Sinusoidal, blendOutTime=1}
    boneToModify {boneName=head}
    scale {y=0.96, z=0.96}
   }
   Save cached pose 'CACHE_Start'(Pose=Component To Local(ComponentPose=@n73)) {cacheName=CACHE_Start}
  ]}
 EventGraph {nodes [
   Event BlueprintUpdateAnimation()
   > n1 = Sequence!(execute) -> (then_0, then_1, then_2:Exec, then_3:Exec, then_4:Exec, then_5:Exec, then_6:Exec, then_7:Exec, then_8:Exec, then_9:Exec)
   n2 = GetAnimInstance(self=GetOwnerSkeletalMeshComponent())
   Is Valid(exec=@n1.then_2, InputObject=@n2)
   > Set EmoteBalletSpinPlaying(EmoteBalletSpinPlaying:bool=Montage_IsPlaying(self=@n2, Montage=/BRCosmetics/Animation/Game/MainPlayer/Emotes/BalletSpin/Emote_BalletSpin_CMF_M)) -> (Output_Get:bool)
   > Is Valid(InputObject=@n2)
   > Set EmoteSunBurstPlaying(EmoteSunBurstPlaying:bool=Montage_IsPlaying(self=@n2, Montage=/BRCosmetics/Animation/Game/MainPlayer/Emotes/SunBurst/Emote_SunBurst_CMF_Skirt_M)) -> (Output_Get:bool)
   n3 = Get bIsMoving2D()
   n4 = GetAlphaFromReferenceTransformDelta(SkeletalMeshComponent=Get Skel Mesh(), BoneName=spine_01, BaseName=root, DeltaComponentType=EulerX, InputRange=(X=0,Y=10), TwistAxis="1,0,0")
   Set RotWithoutMove_L(execute=@n1.then_9, RotWithoutMove_L:bool=AND Boolean(A="float > float"(A=@n4, B=0.95), B=NOT Boolean(A=@n3))) -> (Output_Get:bool)
   n5 = Get bIsMoving2D()
   n6 = GetAlphaFromReferenceTransformDelta(
    SkeletalMeshComponent=Get Skel Mesh()
    BoneName=spine_01
    BaseName=root
    DeltaComponentType=EulerX
    InputRange=(X=0,Y=-55)
    TwistAxis="1,0,0"
   )
   Set RotWithoutMove_R(execute=@n1.then_8, RotWithoutMove_R:bool=AND Boolean(A="float > float"(A=@n6, B=0.99), B=NOT Boolean(A=@n5))) -> (Output_Get:bool)
   Set LOD1(execute=@n1.then_7, LOD1:bool="integer >= integer"(A=Get CurrentLOD(), B=1)) -> (Output_Get:bool)
   Set ASWon(execute=@n1.then_1, ASWon:bool=NOT Boolean(A=Get bRigidBodyEnabled())) -> (Output_Get:bool)
   n7 = Get bIsMoving2D()
   n8 = Get bIsParachuteOpened()
   Set ParachutMoving(execute=@n1.then_5, ParachutMoving:bool=AND Boolean(A=@n7, B=@n8)) -> (Output_Get:bool)
   n9 = Get bIsCrouching()
   Set StandUpMoving(execute=@n1.then_4, StandUpMoving:bool=AND Boolean(A=NOT Boolean(A=@n9), B=@n7)) -> (Output_Get:bool)
   n10 = Gravity Override State Function DEPRECATED(
    execute=@n1.then_0
    CharacterPartAnimInstance=Self-Reference()
    GameTimeInSeconds=GetGameTimeInSeconds()
    GravityOverrideParams=Get Gravity Override Params()
   )
   > Set GravityOverride(GravityOverride=@n10)
   Set SkyNoParachute(execute=@n1.then_6, SkyNoParachute:bool=AND Boolean(A=NOT Boolean(A=@n8), B=Get bIsSkyDiving())) -> (Output_Get:bool)
   n11 = Cast To FortPlayerPawn(execute=@n1.then_3, Object=TryGetPawnOwner())
   n12 = GetSkeletalMeshForPartType(self=@n11, PartType=Face)
   > Set ShortHairEnable(ShortHairEnable:bool=MatchesWildcard(SourceString=GetDisplayName(Object=@n12), Wildcard=*F_MED_ForestBath_Hair_FaceAcc*, SearchCase=CaseSensitive)) -> (Output_Get:bool)
   Event OnParachuteOpened()
   > Set ParachutBeginning(ParachutBeginning:bool=true) -> (Output_Get:bool)
   > Delay(Duration=2)
   > Set ParachutBeginning(ParachutBeginning=false)
   Event BlueprintInitializeAnimation()
   > Set Skel Mesh(Skel Mesh=GetOwnerSkeletalMeshComponent())
   Event OnCharacterCustomizationLoaded()
   > n13 = Cast To PlayerPawn_Athena(Object=TryGetPawnOwner())
   > "Set Body Mesh "("Body Mesh ":Skeletal Mesh Component Object Reference=GetSkeletalMeshForPartType(self=@n13)) -> (Output_Get:Skeletal Mesh Component Object Reference)
   > "Set Face Mesh "("Face Mesh ":Skeletal Mesh Component Object Reference=GetSkeletalMeshForPartType(self=@n13, PartType=Head)) -> (Output_Get:Skeletal Mesh Component Object Reference)
   Event OnWeaponChanged()
   > Set ChangeWeapon(ChangeWeapon:bool=true) -> (Output_Get:bool)
   > Delay(Duration=0.5)
   > Set ChangeWeapon(ChangeWeapon=false)
  ]}
]
variables [
 ASWon {type=boolean}
 "Body Mesh " {type=SkeletalMeshComponent}
 ChangeWeapon {type=boolean}
 EmoteBalletSpinPlaying {type=boolean}
 EmoteSunBurstPlaying {type=boolean}
 "Face Mesh " {type=SkeletalMeshComponent}
 Gravity Override Params {type=GravityOverrideParamsStruct, default {dBNO {gravityOverride=<-980 600 0>}}}
 GravityOverride {type=Vector}
 LOD1 {type=boolean}
 ParachutBeginning {type=boolean}
 ParachutMoving {type=boolean}
 RotWithoutMove_L {type=boolean}
 RotWithoutMove_R {type=boolean}
 ShortHairEnable {type=boolean}
 Skel Mesh {type=SkeletalMeshComponent}
 SkyNoParachute {type=boolean}
 StandUpMoving {type=boolean}
]
