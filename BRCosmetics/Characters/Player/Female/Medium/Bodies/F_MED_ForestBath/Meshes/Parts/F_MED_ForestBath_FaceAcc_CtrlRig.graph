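# Control Rig for the ForestBath female face accessory that procedurally adjusts back hair and earrings to head pose and player state.
# Pelvis height feeds a curve into UpButCrouching, and spherical pose readers on the head weight corrective bone offsets.
# WearingBackpack selects the hair's aim constraint and correctives; without one, crouch and cape states add hair and cape-helper offsets.
# Earring offsets apply while targeting; backpack hair correctives only when not targeting.
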ControlRigBlueprint /BRCosmetics/Characters/Player/Female/Medium/Bodies/F_MED_ForestBath/Meshes/Parts/F_MED_ForestBath_FaceAcc_CtrlRig
graphs [
 Multiply_4_1_1_1_1_1_ContainedGraph {nodes [
   n1 = RigVMFunctionEntryNode()
   RigVMFunctionReturnNode(Result=&RigVMFunction_MathDoubleMul(B=@n1.C, A=&RigVMFunction_MathDoubleMul(A=@n1.A, B=@n1.B).Result).Result)
  ]}
 Multiply_4_1_1_1_1_1_ContainedGraph {nodes [
   n1 = RigVMFunctionEntryNode()
   n2 = &RigVMFunction_MathDoubleMul(A=@n1.A, B=@n1.B)
   RigVMFunctionReturnNode(Result=&RigVMFunction_MathDoubleMul(B=@n1.D, A=&RigVMFunction_MathDoubleMul(B=@n1.C, A=@n2.Result).Result).Result)
  ]}
 RigVMFunctionLibrary {links [], nodes []}
 RigVMFunction_Sequence_1_ContainedGraph {nodes [
   n1 = RigVMFunctionEntryNode()
   RigVMFunction_Sequence = &RigVMFunction_Sequence(ExecuteContext=@n1.ExecuteContext)
   n2 = &RigVMFunction_Sequence(ExecuteContext=@RigVMFunction_Sequence.B)
   n3 = &RigVMFunction_Sequence(ExecuteContext=@n2.B)
   n4 = &RigVMFunction_Sequence(ExecuteContext=@n3.B)
   n5 = &RigVMFunction_Sequence(ExecuteContext=@n4.B)
   RigVMFunctionReturnNode(A=@RigVMFunction_Sequence.A, B=@n2.A, C=@n3.A, D=@n4.A, E=@n5.A, F=@n5.B)
  ]}
 RigVMModel {nodes [
   RigUnit_BeginExecution = &RigUnit_BeginExecution()
   RigVMFunction_Sequence = RigVMAggregateNode(ExecuteContext=@RigUnit_BeginExecution.ExecutePin)
   RigVMFunction_ControlFlowBranch = &RigVMFunction_ControlFlowBranch(Condition=RigVMVariableNode() {variable=WearingBackpack}.Value, ExecuteContext=@RigVMFunction_Sequence.C)
   &RigUnit_AimConstraintLocalSpaceOffset(
    Child=(Type=Bone,Name="base_hair_back")
    Filter=(bX=True,bY=true,bZ=true)
    Parents=((Item=(Type=Bone,Name="Aim_hair01"),Weight=1))
    UpAxis=(X=0,Y=-1,Z=0)
    Weight=1
    WorldUp=(Target=(X=0,Y=-1,Z=0),Kind=Location,Space=(Type=Bone,Name="Aim_hair01"))
    bMaintainOffset=true
    ExecutePin=@RigVMFunction_ControlFlowBranch.False
   )
   n1 = /Script/RigVM.RigVMFunction_MathBoolMake(Value=False) {is=struct RigVMUnitNode}
   > n2 = &RigUnit_SphericalPoseReader(
    ActiveRegionScaleFactors=(PositiveWidth=0.75,NegativeWidth=0.75,PositiveHeight=1,NegativeHeight=1)
    ActiveRegionSize=0.1
    Debug=(bDrawDebug=true,DebugScale=25,DebugSegments=20,DebugThickness=0.25)
    DriverAxis=(X=0,Y=1,Z=0)
    DriverItem=(Type=Bone,Name="head")
    FalloffRegionScaleFactors=(PositiveWidth=1,NegativeWidth=1,PositiveHeight=1,NegativeHeight=2.2)
    FalloffSize=0.2
    OptionalParentItem=(Type=Bone)
    RotationOffset=(X=150,Y=0,Z=0)
    Debug.bDrawDebug=@n1.Value
   )
   > &RigUnit_ModifyTransforms(
    ItemToModify=((Item=(Type=Bone,Name="base_hair_back"),Transform=(Rotation=(X=0.0084,Y=0.3227,Z=-0.0456,W=0.9454),Translation=(X=-0.6141,Y=0.7872,Z=1.7576),Scale3D=(X=1,Y=1,Z=1))))
    WeightMinimum=0
    Weight=@n2.OutputParam
   )
   > n3 = &RigUnit_SphericalPoseReader(
    ActiveRegionScaleFactors=(PositiveWidth=0.75,NegativeWidth=0.75,PositiveHeight=1,NegativeHeight=1)
    ActiveRegionSize=0.1
    Debug=(bDrawDebug=true,DebugScale=25,DebugSegments=20,DebugThickness=0.25)
    DriverAxis=(X=0,Y=1,Z=0)
    DriverItem=(Type=Bone,Name="head")
    FalloffRegionScaleFactors=(PositiveWidth=1,NegativeWidth=1,PositiveHeight=1,NegativeHeight=2.2)
    FalloffSize=0.2
    FlipHeightScaling=true
    FlipWidthScaling=false
    OptionalParentItem=(Type=Bone)
    RotationOffset=(X=30,Y=0,Z=0)
    Debug.bDrawDebug=@n1.Value
   )
   > &RigUnit_ModifyTransforms(
    ItemToModify=((Item=(Type=Bone,Name="base_hair_back"),Transform=(Rotation=(X=-0.0084,Y=-0.3227,Z=-0.0456,W=0.9454),Translation=(X=-0.6141,Y=0.7872,Z=-1.7576),Scale3D=(X=1,Y=1,Z=1))))
    WeightMinimum=0
    Weight=@n3.OutputParam
   )
   > n4 = &RigUnit_SphericalPoseReader(
    ActiveRegionScaleFactors=(PositiveWidth=1,NegativeWidth=1,PositiveHeight=1,NegativeHeight=1)
    ActiveRegionSize=0.04
    Debug=(bDrawDebug=true,DebugScale=25,DebugSegments=20,DebugThickness=0.25)
    DriverAxis=(X=0,Y=1,Z=0)
    DriverItem=(Type=Bone,Name="head")
    FalloffRegionScaleFactors=(PositiveWidth=1,NegativeWidth=1,PositiveHeight=1,NegativeHeight=1)
    FalloffSize=0.1
    OptionalParentItem=(Type=Bone)
    RotationOffset=(X=62,Y=-110,Z=0)
    Debug.bDrawDebug=@n1.Value
   )
   > &RigUnit_ModifyTransforms(
    ItemToModify=((Item=(Type=Bone,Name="base_hair_back"),Transform=(Rotation=(X=0,Y=0.2164,Z=0,W=0.9763),Translation=(X=0,Y=0,Z=0),Scale3D=(X=1,Y=1,Z=1))))
    WeightMinimum=0
    Weight=@n4.OutputParam
   )
   > n5 = &RigUnit_SphericalPoseReader(
    ActiveRegionScaleFactors=(PositiveWidth=3.5,NegativeWidth=0.6,PositiveHeight=1,NegativeHeight=1)
    ActiveRegionSize=0.1
    Debug=(bDrawDebug=true,bDraw2D=False,bDrawLocalAxes=False,DebugScale=25,DebugSegments=20,DebugThickness=0.25)
    DriverAxis=(X=0,Y=1,Z=0)
    DriverItem=(Type=Bone,Name="head")
    FalloffRegionScaleFactors=(PositiveWidth=0.01,NegativeWidth=0.4,PositiveHeight=1,NegativeHeight=1)
    FalloffSize=0.2
    FlipHeightScaling=false
    FlipWidthScaling=false
    OptionalParentItem=(Type=Bone,Name="None")
    RotationOffset=(X=90,Y=0,Z=-35)
    Debug.bDrawDebug=@n1.Value
   )
   n6 = RigVMVariableNode() {variable=IsCrouching}
   > &RigUnit_ModifyTransforms(
    ItemToModify=((Item=(Type=Bone,Name="base_hair_back"),Transform=(Rotation=(X=0,Y=0.2164,Z=0,W=0.9763),Translation=(X=0,Y=0,Z=0),Scale3D=(X=1,Y=1,Z=1))))
    WeightMinimum=0
    Weight=&RigVMFunction_MathDoubleMul(A=&RigVMFunction_MathBoolToFloat(Value=@n6.Value).Result, B=@n5.OutputParam).Result
   )
   n7 = RigVMVariableNode() {variable=UpButCrouching}
   Multiply_4_1_1_1_1_1 = RigVMAggregateNode(
    A=&RigVMFunction_MathBoolToFloat(Value=/Script/RigVM.RigVMFunction_MathDoubleGreater(B=0.35, A=@n7.Value) {is=struct RigVMUnitNode, template="Greater::Execute(in A,in B,out Result)"}.Result).Result
    B=&RigVMFunction_MathBoolToFloat(Value=RigVMVariableNode() {variable=IsCrouching}.Value).Result
    C=@n5.OutputParam
   )
   > &RigUnit_ModifyTransforms(
    ItemToModify=((Item=(Type=Bone,Name="base_hair_back"),Transform=(Rotation=(X=-0.0872,Y=0,Z=0,W=0.9962),Translation=(X=0,Y=1,Z=0),Scale3D=(X=1,Y=1,Z=1))))
    WeightMinimum=0
    Weight=@Multiply_4_1_1_1_1_1.Result
   )
   n8 = RigVMVariableNode() {variable=UpButCrouching}
   Multiply_4_1_1_1_1_1_1 = RigVMAggregateNode(
    D=&RigVMFunction_MathBoolToFloat(Value=&RigVMFunction_MathDoubleGreater(B=0.35, A=@n8.Value) {template="Greater::Execute(in A,in B,out Result)"}.Result).Result
    B=&RigVMFunction_MathBoolToFloat(Value=RigVMVariableNode() {variable=IsCrouching}.Value).Result
    C=&RigVMFunction_MathBoolToFloat(Value=RigVMVariableNode() {variable=WearingCape}.Value).Result
    A=@n5.OutputParam
   )
   > &RigUnit_ModifyTransforms(
    ItemToModify="((Item=(Type=Bone,Name=\"base_cape_help\"),Transform=(Rotation=(X=-0.3827,Y=0,Z=0,W=0.9239),Translation=(X=-5,Y=0,Z=0),Scale3D=(X=1,Y=1,Z=1))),(Item=(Type=Bone,Name=\"base_hair_back\"),Transform=(Rotation=(X=0.0858,Y=0.173,Z=-0.0151,W=0.9811),Translation=(X=1, ...(truncated)"
    WeightMinimum=0
    Weight=@Multiply_4_1_1_1_1_1_1.Result
   )
   &RigUnit_AimConstraintLocalSpaceOffset(
    Child=(Type=Bone,Name="base_hair_back")
    Filter=(bX=True,bY=True,bZ=True)
    Parents=((Item=(Type=Bone,Name="base_hair_back_aimtarget"),Weight=1))
    UpAxis=(X=0,Y=1,Z=0)
    Weight=1
    WorldUp=(Target=(X=0,Y=0,Z=0),Kind=Direction,Space=(Type=None,Name="None"))
    ExecutePin=@RigVMFunction_ControlFlowBranch.True
   )
   n9 = &RigVMFunction_MathBoolMake(Value=False)
   > n10 = &RigUnit_SphericalPoseReader(
    ActiveRegionScaleFactors=(PositiveWidth=0.75,NegativeWidth=0.75,PositiveHeight=1,NegativeHeight=0.5)
    ActiveRegionSize=0.1
    Debug=(bDrawDebug=false,bDraw2D=False,bDrawLocalAxes=False,DebugScale=25,DebugSegments=20,DebugThickness=0.25)
    DriverAxis=(X=0,Y=1,Z=0)
    DriverItem=(Type=Bone,Name="head")
    FalloffRegionScaleFactors=(PositiveWidth=1,NegativeWidth=1,PositiveHeight=1,NegativeHeight=0.5)
    FalloffSize=0.2
    FlipHeightScaling=false
    FlipWidthScaling=False
    OptionalParentItem=(Type=Bone,Name="None")
    RotationOffset=(X=120,Y=0,Z=0)
    Debug.bDrawDebug=@n9.Value
   )
   n11 = &RigVMFunction_MathBoolToFloat(Value=&RigVMFunction_MathBoolNot(Value=RigVMVariableNode() {variable=IsTargeting}.Value).Result)
   > &RigUnit_ModifyTransforms(
    ItemToModify=((Item=(Type=Bone,Name="base_hair_back"),Transform=(Rotation=(X=-0.0868,Y=-0.0868,Z=-0.0076,W=0.9924),Translation=(X=0,Y=0,Z=0),Scale3D=(X=1,Y=1,Z=1))))
    WeightMinimum=0
    Weight=&RigVMFunction_MathDoubleMul(B=@n11.Result, A=@n10.OutputParam).Result
   )
   > n12 = &RigUnit_SphericalPoseReader(
    ActiveRegionScaleFactors=(PositiveWidth=0.75,NegativeWidth=0.75,PositiveHeight=1,NegativeHeight=0.5)
    ActiveRegionSize=0.1
    Debug=(bDrawDebug=false,bDraw2D=False,bDrawLocalAxes=False,DebugScale=25,DebugSegments=20,DebugThickness=0.25)
    DriverAxis=(X=0,Y=1,Z=0)
    DriverItem=(Type=Bone,Name="head")
    FalloffRegionScaleFactors=(PositiveWidth=1,NegativeWidth=1,PositiveHeight=1,NegativeHeight=0.5)
    FalloffSize=0.2
    FlipHeightScaling=false
    FlipWidthScaling=False
    OptionalParentItem=(Type=Bone,Name="None")
    RotationOffset=(X=120,Y=50,Z=0)
    Debug.bDrawDebug=@n9.Value
   )
   > &RigUnit_ModifyTransforms(
    ItemToModify=((Item=(Type=Bone,Name="base_hair_back"),Transform=(Rotation=(X=-0.173,Y=-0.0858,Z=-0.0151,W=0.9811),Translation=(X=0,Y=0,Z=0),Scale3D=(X=1,Y=1,Z=1))))
    WeightMinimum=0
    Weight=&RigVMFunction_MathDoubleMul(A=@n12.OutputParam, B=@n11.Result).Result
   )
   RigUnit_GetRelativeTransformForItem = &RigUnit_GetRelativeTransformForItem(Child=(Type=Bone,Name="pelvis"), Parent=(Type=Bone,Name="root"), bChildInitial=False, bParentInitial=False)
   Remap = &RigVMFunction_MathDoubleRemap(
    SourceMaximum=40
    SourceMinimum=70
    TargetMaximum=0
    TargetMinimum=1
    bClamp=False
    Value=@RigUnit_GetRelativeTransformForItem.RelativeTransform.Translation.Z
   )
   n13 = &RigVMFunction_AnimEvalRichCurve(
    Curve="(EditorCurveData=(Keys=((InterpMode=RCIM_Cubic,TangentMode=RCTM_Break,TangentWeightMode=RCTWM_WeightedNone,Time=0,Value=0,ArriveTangent=0,ArriveTangentWeight=0,LeaveTangent=0,LeaveTangentWeight=0),(InterpMode=RCIM_Cubic,TangentMode=RCTM_Break,TangentWeight ...(truncated)"
    SourceMinimum=0
    TargetMinimum=0
    Value=@Remap.Result
   )
   RigVMVariableNode(Value=@n13.Result, ExecuteContext=@RigVMFunction_Sequence.A) {variable=UpButCrouching}
   RigVMFunction_MathBoolMake = &RigVMFunction_MathBoolMake(Value=False)
   n14 = &RigUnit_SphericalPoseReader(
    ActiveRegionScaleFactors=(PositiveWidth=1.5,NegativeWidth=0.8,PositiveHeight=1,NegativeHeight=1)
    ActiveRegionSize=0.1
    Debug=(bDrawDebug=false,bDraw2D=False,bDrawLocalAxes=False,DebugScale=25,DebugSegments=20,DebugThickness=0.25)
    DriverAxis=(X=0,Y=1,Z=0)
    DriverItem=(Type=Bone,Name="head")
    FalloffRegionScaleFactors=(PositiveWidth=1,NegativeWidth=0.35,PositiveHeight=1,NegativeHeight=1)
    FalloffSize=0.2
    FlipHeightScaling=False
    FlipWidthScaling=False
    OptionalParentItem=(Type=Bone,Name="None")
    RotationOffset=(X=90,Y=0,Z=-30)
    Debug.bDrawDebug=@RigVMFunction_MathBoolMake.Value
    ExecutePin=@RigVMFunction_Sequence.B
   )
   n15 = &RigVMFunction_MathBoolToFloat(Value=RigVMVariableNode() {variable=IsTargeting}.Value)
   > &RigUnit_ModifyTransforms(
    ItemToModify="((Item=(Type=Bone,Name=\"base_earrin_r\"),Transform=(Rotation=(X=-0.2156,Y=-0.0189,Z=0.0851,W=0.9726),Translation=(X=0,Y=0,Z=0),Scale3D=(X=1,Y=1,Z=1))),(Item=(Type=Bone,Name=\"base_earrin_l\"),Transform=(Rotation=(X=0,Y=0,Z=0.1736,W=0.9848),Translation=(X=0,Y= ...(truncated)"
    WeightMinimum=0
    Weight=&RigVMFunction_MathDoubleMul(B=@n15.Result, A=@n14.OutputParam).Result
   )
   > &RigUnit_ModifyTransforms(
    ItemToModify=((Item=(Type=Bone,Name="base_earrin_l"),Transform=(Rotation=(X=0.5,Y=0,Z=0,W=0.866),Translation=(X=0,Y=0,Z=0),Scale3D=(X=1,Y=1,Z=1))))
    WeightMinimum=0
    Weight=@n15.Result
   )
   > n16 = &RigUnit_SphericalPoseReader(
    ActiveRegionScaleFactors=(PositiveWidth=1.1,NegativeWidth=0.5,PositiveHeight=1,NegativeHeight=1)
    ActiveRegionSize=0.1
    Debug=(bDrawDebug=false,bDraw2D=False,bDrawLocalAxes=False,DebugScale=25,DebugSegments=20,DebugThickness=0.25)
    DriverAxis=(X=0,Y=1,Z=0)
    DriverItem=(Type=Bone,Name="head")
    FalloffRegionScaleFactors=(PositiveWidth=1,NegativeWidth=0.5,PositiveHeight=1,NegativeHeight=1)
    FalloffSize=0.2
    FlipHeightScaling=False
    FlipWidthScaling=False
    OptionalParentItem=(Type=Bone,Name="None")
    RotationOffset=(X=90,Y=0,Z=20)
    Debug.bDrawDebug=@RigVMFunction_MathBoolMake.Value
   )
   n17 = RigVMVariableNode() {variable=IsTargeting}
   > &RigUnit_ModifyTransforms(
    ItemToModify="((Item=(Type=Bone,Name=\"base_earrin_r\"),Transform=(Rotation=(X=-0.3536,Y=0.3536,Z=0.1464,W=0.8536),Translation=(X=0,Y=0,Z=0),Scale3D=(X=1,Y=1,Z=1))),(Item=(Type=Bone,Name=\"base_earrin_l\"),Transform=(Rotation=(X=0,Y=-0.3827,Z=0,W=0.9239),Translation=(X=0,Y= ...(truncated)"
    WeightMinimum=0
    Weight=&RigVMFunction_MathDoubleMul(B=&RigVMFunction_MathBoolToFloat(Value=@n17.Value).Result, A=@n16.OutputParam).Result
   )
   RigUnit_GetTransform = &RigUnit_GetTransform(Item=(Type=Control,Name="head_ctrl"), bInitial=False)
   RigUnit_SetTransform = &RigUnit_SetTransform(ExecutePin=(), Item=(Type=Bone,Name="head"), Weight=1, bInitial=False, Value=@RigUnit_GetTransform.Transform)
  ]}
]
hierarchy [
 root {loc=<0 0 0>, rot=<0 0 0 1>}
 pelvis {parent=root, loc=<0 -0.5301 98.0571>, rot=<0.0127 0.707 -0.0127 -0.707>}
 spine_01 {parent=pelvis, loc=<6.1208 -0.9328 0>, rot=<0 0 0.034 -0.9994>}
 spine_02 {parent=spine_01, loc=<6.5997 -0.2095 0>, rot=<0 0 0.0238 -0.9997>}
 spine_03 {parent=spine_02, loc=<6.4914 -0.1237 0>, rot=<0 0 0.0149 -0.9999>}
 spine_04 {parent=spine_03, loc=<6.9598 0.0794 0>, rot=<0 0 -0.0132 -0.9999>}
 spine_05 {parent=spine_04, loc=<6.9987 0 0>, rot=<0 0 -0.0015 -1>}
 neck_01 {parent=spine_05, loc=<11.3138 0.1529 0>, rot=<0 0 -0.2003 -0.9797>}
 neck_02 {parent=neck_01, loc=<4.5627 0 0>, rot=<0 0 0.0504 -0.9987>}
 head {parent=neck_02, loc=<5.3914 0 0>, rot=<0 0 0.115 -0.9934>}
 base_hair_back {parent=head, loc=<11.4312 -10.3363 0>, rot=<0.4413 -0.8973 0 0>}
 base_hair_back_l {parent=base_hair_back, loc=<8.7736 0.0009 9.5>, rot=<0 0 -0.2583 -0.9661>}
 dyn_hair_back_l_01 {parent=base_hair_back_l, loc=<0 0 0>, rot=<0 0 0 1>}
 dyn_hair_back_l_02 {parent=dyn_hair_back_l_01, loc=<10.0875 0.0041 0>, rot=<0 0 -0.1893 -0.9819>}
 dyn_hair_back_l_03 {parent=dyn_hair_back_l_02, loc=<11.1601 0.0288 0>, rot=<0 0 -0.0434 -0.9991>}
 dyn_hair_back_l_04 {parent=dyn_hair_back_l_03, loc=<10.6841 -0.0126 0>, rot=<0 0 0 1>}
 base_hair_back_r {parent=base_hair_back, loc=<8.7736 0.0009 -9.5>, rot=<0 0 -0.2583 -0.9661>}
 dyn_hair_back_r_01 {parent=base_hair_back_r, loc=<0 0 0>, rot=<0 0 0 1>}
 dyn_hair_back_r_02 {parent=dyn_hair_back_r_01, loc=<10.0875 0.0041 0>, rot=<0 0 -0.1893 -0.9819>}
 dyn_hair_back_r_03 {parent=dyn_hair_back_r_02, loc=<11.1601 0.0288 0>, rot=<0 0 -0.0434 -0.9991>}
 dyn_hair_back_r_04 {parent=dyn_hair_back_r_03, loc=<10.6841 -0.0126 0>, rot=<0 0 0 1>}
 base_hair_back_m {parent=base_hair_back, loc=<8.7736 0.0009 0>, rot=<0 0 -0.2583 -0.9661>}
 dyn_hair_back_m_01 {parent=base_hair_back_m, loc=<0 0 0>, rot=<0 0 0 1>}
 dyn_hair_back_m_02 {parent=dyn_hair_back_m_01, loc=<10.0875 0.0041 0>, rot=<0 0 -0.1893 -0.9819>}
 dyn_hair_back_m_03 {parent=dyn_hair_back_m_02, loc=<11.1601 0.0288 0>, rot=<0 0 -0.0434 -0.9991>}
 dyn_hair_back_m_04 {parent=dyn_hair_back_m_03, loc=<10.6841 -0.0126 0>, rot=<0 0 0 1>}
 base_hair_front_r {parent=head, loc=<2.16 4.831 6.9676>, rot=<-0.1329 -0.5439 0.8285 -0.0117>}
 dyn_hair_front_r {parent=base_hair_front_r, loc=<0 0 0>, rot=<0 0 0 1>}
 base_hair_front_l {parent=head, loc=<2.1843 3.4224 -7.7932>, rot=<0.1196 0.6203 0.771 -0.0802>}
 dyn_hair_front_l {parent=base_hair_front_l, loc=<0 0 0>, rot=<0 0 0 1>}
 base_earrin_l {parent=head, loc=<3.8648 -1.674 -9.5181>, rot=<-0.0012 0.3007 0.9537 -0.0039>}
 dyn_earrin_l {parent=base_earrin_l, loc=<0 0 0>, rot=<0 0 0 1>}
 base_earrin_r {parent=head, loc=<3.8646 -1.6602 9.526>, rot=<0.0001 -0.2588 0.9659 -0.0078>}
 dyn_earrin_r {parent=base_earrin_r, loc=<0 0 0>, rot=<0 0 0 1>}
 base_hair_top_front_r {parent=head, loc=<15.1889 5.301 0.9921>, rot=<0.4783 0.4056 -0.0258 -0.7785>}
 dyn_hair_top_front_r {parent=base_hair_top_front_r, loc=<0 0 0>, rot=<0 0 0 1>}
 base_hair_top_front_l {parent=head, loc=<16.1997 4.6257 -1.2971>, rot=<-0.7092 -0.0868 0.1637 -0.6802>}
 dyn_hair_top_front_l {parent=base_hair_top_front_l, loc=<0 0 0>, rot=<0 0 0 1>}
 base_neck_front_r {parent=neck_01, loc=<-4.7431 11.9964 9.0679>, rot=<-0.0816 -0.0102 -0.2922 -0.9528>}
 dyn_neck_front_r {parent=base_neck_front_r, loc=<0 0 0>, rot=<0 0 0 1>}
 base_neck_mid_r {parent=neck_01, loc=<-0.109 1.7768 7.1023>, rot=<-0.1233 0.0033 -0.0134 -0.9923>}
 dyn_neck_mid_r {parent=base_neck_mid_r, loc=<0 0 0>, rot=<0 0 0 1>}
 base_neck_midback_r {parent=neck_01, loc=<-0.228 -2.6888 4.9812>, rot=<-0.3592 0 -0.0138 -0.9332>}
 dyn_neck_midback_r {parent=base_neck_midback_r, loc=<0 0 0>, rot=<0 0 0 1>}
 base_neck_back {parent=neck_01, loc=<-0.6643 -4.8913 -0.0562>, rot=<0 0 0.0401 -0.9992>}
 dyn_neck_back {parent=base_neck_back, loc=<0 0 0>, rot=<0 0 0 1>}
 base_neck_mid_l {parent=neck_01, loc=<-0.1213 1.7816 -7.1559>, rot=<0.1365 0.0066 -0.0121 -0.9906>}
 dyn_neck_mid_l {parent=base_neck_mid_l, loc=<0 0 0>, rot=<0 0 0 1>}
 base_neck_front_l {parent=neck_01, loc=<-4.7564 12.0188 -8.9018>, rot=<0.0728 -0.0148 -0.2873 -0.9549>}
 dyn_neck_front_l {parent=base_neck_front_l, loc=<0 0 0>, rot=<0 0 0 1>}
 clavicle_l {parent=spine_05, loc=<5.4831 -0.1053 -3.2946>, rot=<0.0465 -0.7289 -0.009 -0.6829>}
 upperarm_l {parent=clavicle_l, loc=<9.8658 -1.2136 -0.4105>, rot=<0.0164 -0.3245 -0.0269 -0.9454>}
 lowerarm_l {parent=upperarm_l, loc=<26.465 -0.8345 0.259>, rot=<0.0036 0.0161 -0.2298 -0.9731>}
 upperarm_twist_01_l {parent=upperarm_l, loc=<7.2135 -0.1134 0>, rot=<0 0 0 1>}
 deform_upperarm_bk_base_l {parent=upperarm_twist_01_l, loc=<-6.4053 -4.3746 0.314>, rot=<0 0 0 -1>}
 deform_upperarm_bk_l {parent=deform_upperarm_bk_base_l, loc=<0 0 0>, rot=<0 0 0 1>}
 deform_upperarm_up_base_l {parent=upperarm_twist_01_l, loc=<-8.891 0.4825 3.9517>, rot=<0 0 0 -1>}
 deform_upperarm_up_l {parent=deform_upperarm_up_base_l, loc=<0 0 0>, rot=<0 0 0 1>}
 clavicle_r {parent=spine_05, loc=<5.4831 -0.1053 3.2946>, rot=<0.7289 0.0465 -0.6829 0.009>}
 upperarm_r {parent=clavicle_r, loc=<-9.8658 1.2136 0.4105>, rot=<0.0164 -0.3245 -0.0269 -0.9454>}
 upperarm_twist_01_r {parent=upperarm_r, loc=<-7.2135 0.1134 0>, rot=<0 0 0 -1>}
 deform_upperarm_bk_base_r {parent=upperarm_twist_01_r, loc=<6.4053 4.3746 -0.314>, rot=<0 0 0 -1>}
 deform_upperarm_bk_r {parent=deform_upperarm_bk_base_r, loc=<0 0 0>, rot=<0 0 0 1>}
 deform_upperarm_up_base_r {parent=upperarm_twist_01_r, loc=<8.891 -0.4825 -3.9517>, rot=<0 0 0 -1>}
 deform_upperarm_up_r {parent=deform_upperarm_up_base_r, loc=<0 0 0>, rot=<0 0 0 1>}
 base_hair_back_aimtarget {parent=spine_05, loc=<8 -10 0>, rot=<0.0284 -0.7065 -0.0284 -0.7065>}
 Aim_hair01 {parent=spine_05, loc=<0 -11.1302 0>, rot=<-0.0284 0.7065 0.0284 0.7065>}
 base_hair_back_aimtarget1 {parent=spine_05, loc=<0 0 0>, rot=<0.7349 -0.6782 0 0>}
 base_cape_help {parent=spine_05, loc=<0 0 0>, rot=<0.7349 -0.6782 0 0>}
 head_ctrl {controlType=EULER_TRANSFORM, type=Control, loc=<0 0 0>, rot=<0 0 0 1>}
]
variables [
 IsCrouching {type=bool, default=False}
 IsMoving2D {type=bool, default=False}
 IsTargeting {type=bool, default=False}
 UpButCrouching {type=double, default=0}
 WearingBackpack {type=bool, default=False}
 WearingCape {type=bool, default=False}
]
