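# Physics asset for the F_MED_ForestBath face accessory mesh, set up for secondary motion on hair, earrings and neck pieces.
# Capsules cover the head, upper spine, clavicles, arms (no right forearm) and a cape helper, alongside the accessory bones.
# Base anchor bones link to head or spine_05 with free angular constraints.
# Dyn bones hang from them with angle-limited, low-twist constraints, including three four-link back-hair chains and single-link front hair, earrings and neck pieces.
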PhysicsAsset /BRCosmetics/Characters/Player/Female/Medium/Bodies/F_MED_ForestBath/Meshes/Parts/F_MED_ForestBath_FaceAcc_Physics
bodies {
 base_cape_help [{shapeType=Capsule, length=33, radius=16, shapeName=spine_05_capsule, center=<0 9 0>, rotation=<rot 0 0 90>}]
 base_earrin_l [{shapeType=Capsule, length=0.505, radius=0.505, shapeName=base_earrin_l_capsule, rotation=<rot 0 -90 0>}]
 base_earrin_r [{shapeType=Capsule, length=0.505, radius=0.505, shapeName=base_earrin_r_capsule, rotation=<rot 0 -90 0>}]
 base_hair_back_l [{shapeType=Capsule, length=6, radius=4, shapeName=base_hair_back_l_capsule}]
 base_hair_back_m [{shapeType=Capsule, length=0.505, radius=0.505, shapeName=base_hair_back_m_capsule, rotation=<rot 0 -90 0>}]
 base_hair_back_r [{shapeType=Capsule, length=0.505, radius=0.505, shapeName=base_hair_back_r_capsule, rotation=<rot 0 -90 0>}]
 base_hair_front_l [{shapeType=Capsule, length=0.505, radius=0.505, shapeName=base_hair_front_l_capsule, rotation=<rot 0 -90 0>}]
 base_hair_front_r [{shapeType=Capsule, length=0.505, radius=0.505, shapeName=base_hair_front_r_capsule, rotation=<rot 0 -90 0>}]
 base_neck_back [{shapeType=Capsule, length=0.505, radius=0.505, shapeName=base_neck_back_capsule, rotation=<rot 0 -90 0>}]
 base_neck_front_l [{shapeType=Capsule, length=0.505, radius=0.505, shapeName=base_neck_front_l_capsule, rotation=<rot 0 -90 0>}]
 base_neck_front_r [{shapeType=Capsule, length=0.505, radius=0.505, shapeName=base_neck_front_r_capsule, rotation=<rot 0 -90 0>}]
 base_neck_mid_l [{shapeType=Capsule, length=0.505, radius=0.505, shapeName=base_neck_mid_l_capsule, rotation=<rot 0 -90 0>}]
 base_neck_mid_r [{shapeType=Capsule, length=0.505, radius=0.505, shapeName=base_neck_mid_r_capsule, rotation=<rot 0 -90 0>}]
 base_neck_midback_r [{shapeType=Capsule, length=0.505, radius=0.505, shapeName=base_neck_midback_r_capsule, rotation=<rot 0 -90 0>}]
 clavicle_l [{shapeType=Capsule, length=4, radius=7.5, shapeName=clavicle_l_capsule, center=<5.7043 0 -0.9159>, rotation=<rot 90 0 0>}]
 clavicle_r [{shapeType=Capsule, length=5, radius=6.5, shapeName=clavicle_l_capsule, center=<-6 0 1>, rotation=<rot 85 0 0>}]
 dyn_earrin_l [{shapeType=Capsule, length=7, radius=2, shapeName=dyn_earrin_l_capsule, center=<6 0 -0.5>, rotation=<rot 85 180 180>}]
 dyn_earrin_r [{shapeType=Capsule, length=7, radius=2, shapeName=dyn_earrin_r_capsule, center=<6 0 0.5>, rotation=<rot 85 0 0>}]
 dyn_hair_back_l_01 [{shapeType=Capsule, length=6, radius=8.09, shapeName=dyn_hair_back_l_01_capsule, center=<5 2 0>, rotation=<rot 90 0 0>}]
 dyn_hair_back_l_02 [{shapeType=Capsule, length=6, radius=7, shapeName=dyn_hair_back_l_02_capsule, center=<5 2.5 0>, rotation=<rot 90 0 0>}]
 dyn_hair_back_l_03 [{shapeType=Capsule, length=6, radius=6, shapeName=dyn_hair_back_l_03_capsule, center=<5 0.5 0>, rotation=<rot 90 0 0>}]
 dyn_hair_back_l_04 [{shapeType=Capsule, length=6, radius=6, shapeName=dyn_hair_back_l_04_capsule, center=<5 -0.5 0>}]
 dyn_hair_back_m_01 [{shapeType=Capsule, length=6, radius=8.09, shapeName=dyn_hair_back_m_01_capsule, center=<5 2 0>, rotation=<rot 90 0 0>}]
 dyn_hair_back_m_02 [{shapeType=Capsule, length=6, radius=7, shapeName=dyn_hair_back_m_02_capsule, center=<5 2.5 0>, rotation=<rot 90 0 0>}]
 dyn_hair_back_m_03 [{shapeType=Capsule, length=6, radius=6, shapeName=dyn_hair_back_m_03_capsule, center=<5 0.5 0>, rotation=<rot 90 0 0>}]
 dyn_hair_back_m_04 [{shapeType=Capsule, length=6, radius=6, shapeName=dyn_hair_back_m_04_capsule, center=<5 -0.5 0>}]
 dyn_hair_back_r_01 [{shapeType=Capsule, length=6, radius=8.09, shapeName=dyn_hair_back_r_01_capsule, center=<5 2 -1>, rotation=<rot 90 0 0>}]
 dyn_hair_back_r_02 [{shapeType=Capsule, length=6, radius=7, shapeName=dyn_hair_back_r_02_capsule, center=<5 2.5 -1.5>, rotation=<rot 90 0 0>}]
 dyn_hair_back_r_03 [{shapeType=Capsule, length=6, radius=6, shapeName=dyn_hair_back_r_03_capsule, center=<5 0.5 0>, rotation=<rot 90 0 0>}]
 dyn_hair_back_r_04 [{shapeType=Capsule, length=6, radius=6, shapeName=dyn_hair_back_r_04_capsule, center=<5 -0.5 0>}]
 dyn_hair_front_l [{shapeType=Capsule, length=13, radius=1.3, shapeName=dyn_hair_front_l_capsule, center=<6 -0.2 -1.6>, rotation=<rot 90 0 0>}]
 dyn_hair_front_r [{shapeType=Capsule, length=5, radius=1, shapeName=dyn_hair_front_r_capsule, center=<9.2 0.5 2>, rotation=<rot 0 0 10>}]
 dyn_neck_back [{shapeType=Capsule, length=6, radius=1.5, shapeName=dyn_neck_back_capsule, center=<2.3 0.3 0>}]
 dyn_neck_front_l [{shapeType=Capsule, length=11, radius=0.6, shapeName=dyn_neck_front_l_capsule, center=<3 -4.2 0>, rotation=<rot 0 -18 90>}]
 dyn_neck_front_r [{shapeType=Capsule, length=11, radius=0.6, shapeName=dyn_neck_front_l_capsule, center=<2.8 -4.2 0>, rotation=<rot 0 -17 90>}]
 dyn_neck_mid_l [{shapeType=Capsule, length=11, radius=0.6, shapeName=dyn_neck_front_l_capsule, center=<2 -1.85 0.5>, rotation=<rot 0 5 92>}]
 dyn_neck_mid_r [{shapeType=Capsule, length=9, radius=0.6, shapeName=dyn_neck_front_l_capsule, center=<1.7 -1.7 -0.25>, rotation=<rot 0 0 90>}]
 dyn_neck_midback_r [{shapeType=Capsule, length=4, radius=1.5, shapeName=dyn_neck_midback_r_capsule, center=<1.8 0 0>, rotation=<rot 0 0 90>}]
 head [{shapeType=Capsule, length=8, radius=8, shapeName=head_capsule, center=<4 0 0.0067>, rotation=<rot 0 58.2958 -89.8606>}]
 lowerarm_l [{shapeType=Capsule, length=19, radius=7, shapeName=lowerarm_l_capsule, center=<9 0 0>, rotation=<rot 0 90 90>}]
 spine_03 [{shapeType=Capsule, length=22, radius=11.2, shapeName=spine_04_capsule, center=<-15 -3.5 0>, rotation=<rot 90 0 -15>}]
 spine_04 [{shapeType=Capsule, length=11, radius=9.8, shapeName=spine_04_capsule, center=<2 0 0>}]
 spine_05 [{shapeType=Capsule, length=11, radius=11, shapeName=spine_05_capsule, center=<4.3 2.5 0>}]
 upperarm_l [{shapeType=Capsule, length=16, radius=7, shapeName=upperarm_l_capsule, center=<11 0 -0.5>, rotation=<rot 90 0 0>}]
 upperarm_r [{shapeType=Capsule, length=16, radius=6.5, shapeName=upperarm_r_capsule, center=<-12 0 0>, rotation=<rot 95 0 0>}]
}
constraints [
 {
  bone1Name=base_hair_back_l
  bone2Name=head
  swing1LimitDegrees=45
  swing1Motion=Free
  swing2LimitDegrees=45
  swing2Motion=Free
  twistLimitDegrees=45
  twistMotion=Free
 }
 {bone1Name=dyn_hair_back_l_01, bone2Name=base_hair_back_l, swing1LimitDegrees=15, swing2LimitDegrees=2, twistLimitDegrees=1}
 {bone1Name=dyn_hair_back_l_02, bone2Name=dyn_hair_back_l_01, swing1LimitDegrees=25, swing2LimitDegrees=5, twistLimitDegrees=1}
 {bone1Name=dyn_hair_back_l_03, bone2Name=dyn_hair_back_l_02, swing1LimitDegrees=30, swing2LimitDegrees=8, twistLimitDegrees=1}
 {bone1Name=dyn_hair_back_l_04, bone2Name=dyn_hair_back_l_03, swing1LimitDegrees=20, swing2LimitDegrees=10, twistLimitDegrees=1}
 {
  bone1Name=base_hair_back_m
  bone2Name=head
  swing1LimitDegrees=45
  swing1Motion=Free
  swing2LimitDegrees=45
  swing2Motion=Free
  twistLimitDegrees=45
  twistMotion=Free
 }
 {bone1Name=dyn_hair_back_m_01, bone2Name=base_hair_back_m, swing1LimitDegrees=15, swing2LimitDegrees=2, twistLimitDegrees=1}
 {bone1Name=dyn_hair_back_m_02, bone2Name=dyn_hair_back_m_01, swing1LimitDegrees=25, swing2LimitDegrees=5, twistLimitDegrees=1}
 {bone1Name=dyn_hair_back_m_03, bone2Name=dyn_hair_back_m_02, swing1LimitDegrees=30, swing2LimitDegrees=8, twistLimitDegrees=1}
 {bone1Name=dyn_hair_back_m_04, bone2Name=dyn_hair_back_m_03, swing1LimitDegrees=20, swing2LimitDegrees=10, twistLimitDegrees=1}
 {
  bone1Name=base_hair_back_r
  bone2Name=head
  swing1LimitDegrees=45
  swing1Motion=Free
  swing2LimitDegrees=45
  swing2Motion=Free
  twistLimitDegrees=45
  twistMotion=Free
 }
 {bone1Name=dyn_hair_back_r_01, bone2Name=base_hair_back_r, swing1LimitDegrees=15, swing2LimitDegrees=2, twistLimitDegrees=1}
 {bone1Name=dyn_hair_back_r_02, bone2Name=dyn_hair_back_r_01, swing1LimitDegrees=25, swing2LimitDegrees=5, twistLimitDegrees=1}
 {bone1Name=dyn_hair_back_r_03, bone2Name=dyn_hair_back_r_02, swing1LimitDegrees=30, swing2LimitDegrees=8, twistLimitDegrees=1}
 {bone1Name=dyn_hair_back_r_04, bone2Name=dyn_hair_back_r_03, swing1LimitDegrees=20, swing2LimitDegrees=10, twistLimitDegrees=1}
 {bone1Name=dyn_hair_front_l, bone2Name=base_hair_front_l, swing1LimitDegrees=15, swing2LimitDegrees=20, twistLimitDegrees=1}
 {bone1Name=dyn_hair_front_r, bone2Name=base_hair_front_r, swing1LimitDegrees=15, swing2LimitDegrees=20, twistLimitDegrees=1}
 {bone1Name=dyn_earrin_l, bone2Name=base_earrin_l, swing1LimitDegrees=35, swing2LimitDegrees=15, twistLimitDegrees=1}
 {bone1Name=dyn_earrin_r, bone2Name=base_earrin_r, swing1LimitDegrees=35, swing2LimitDegrees=15, twistLimitDegrees=1}
 {
  bone1Name=base_hair_front_r
  bone2Name=head
  swing1LimitDegrees=45
  swing1Motion=Free
  swing2LimitDegrees=45
  swing2Motion=Free
  twistLimitDegrees=45
  twistMotion=Free
 }
 {
  bone1Name=base_hair_front_l
  bone2Name=head
  swing1LimitDegrees=45
  swing1Motion=Free
  swing2LimitDegrees=45
  swing2Motion=Free
  twistLimitDegrees=45
  twistMotion=Free
 }
 {
  bone1Name=base_earrin_l
  bone2Name=head
  swing1LimitDegrees=45
  swing1Motion=Free
  swing2LimitDegrees=45
  swing2Motion=Free
  twistLimitDegrees=45
  twistMotion=Free
 }
 {
  bone1Name=base_earrin_r
  bone2Name=head
  swing1LimitDegrees=45
  swing1Motion=Free
  swing2LimitDegrees=45
  swing2Motion=Free
  twistLimitDegrees=45
  twistMotion=Free
 }
 {
  bone1Name=upperarm_l
  bone2Name=clavicle_l
  swing1LimitDegrees=45
  swing1Motion=Free
  swing2LimitDegrees=45
  swing2Motion=Free
  twistLimitDegrees=45
  twistMotion=Free
 }
 {
  bone1Name=upperarm_r
  bone2Name=clavicle_r
  swing1LimitDegrees=45
  swing1Motion=Free
  swing2LimitDegrees=45
  swing2Motion=Free
  twistLimitDegrees=45
  twistMotion=Free
 }
 {
  bone1Name=head
  bone2Name=spine_05
  swing1LimitDegrees=45
  swing1Motion=Free
  swing2LimitDegrees=45
  swing2Motion=Free
  twistLimitDegrees=45
  twistMotion=Free
 }
 {
  bone1Name=clavicle_l
  bone2Name=spine_05
  swing1LimitDegrees=45
  swing1Motion=Free
  swing2LimitDegrees=45
  swing2Motion=Free
  twistLimitDegrees=45
  twistMotion=Free
 }
 {
  bone1Name=clavicle_r
  bone2Name=spine_05
  swing1LimitDegrees=45
  swing1Motion=Free
  swing2LimitDegrees=45
  swing2Motion=Free
  twistLimitDegrees=45
  twistMotion=Free
 }
 {
  bone1Name=spine_05
  bone2Name=spine_04
  swing1LimitDegrees=45
  swing1Motion=Free
  swing2LimitDegrees=45
  swing2Motion=Free
  twistLimitDegrees=45
  twistMotion=Free
 }
 {
  bone1Name=base_neck_back
  bone2Name=spine_05
  swing1LimitDegrees=45
  swing1Motion=Free
  swing2LimitDegrees=45
  swing2Motion=Free
  twistLimitDegrees=45
  twistMotion=Free
 }
 {bone1Name=dyn_neck_back, bone2Name=base_neck_back, swing1LimitDegrees=15, swing2LimitDegrees=5, twistLimitDegrees=1}
 {
  bone1Name=base_neck_front_l
  bone2Name=spine_05
  swing1LimitDegrees=45
  swing1Motion=Free
  swing2LimitDegrees=45
  swing2Motion=Free
  twistLimitDegrees=45
  twistMotion=Free
 }
 {bone1Name=dyn_neck_front_l, bone2Name=base_neck_front_l, swing1LimitDegrees=0, swing2LimitDegrees=20, twistLimitDegrees=1}
 {
  bone1Name=base_neck_front_r
  bone2Name=spine_05
  swing1LimitDegrees=45
  swing1Motion=Free
  swing2LimitDegrees=45
  swing2Motion=Free
  twistLimitDegrees=45
  twistMotion=Free
 }
 {bone1Name=dyn_neck_front_r, bone2Name=base_neck_front_r, swing1LimitDegrees=0, swing2LimitDegrees=20, twistLimitDegrees=1}
 {
  bone1Name=base_neck_mid_l
  bone2Name=spine_05
  swing1LimitDegrees=45
  swing1Motion=Free
  swing2LimitDegrees=45
  swing2Motion=Free
  twistLimitDegrees=45
  twistMotion=Free
 }
 {bone1Name=dyn_neck_mid_l, bone2Name=base_neck_mid_l, swing1LimitDegrees=5, swing2LimitDegrees=15, twistLimitDegrees=1}
 {
  bone1Name=base_neck_mid_r
  bone2Name=spine_05
  swing1LimitDegrees=45
  swing1Motion=Free
  swing2LimitDegrees=45
  swing2Motion=Free
  twistLimitDegrees=45
  twistMotion=Free
 }
 {bone1Name=dyn_neck_mid_r, bone2Name=base_neck_mid_r, swing1LimitDegrees=5, swing2LimitDegrees=15, twistLimitDegrees=1}
 {
  bone1Name=base_neck_midback_r
  bone2Name=spine_05
  swing1LimitDegrees=45
  swing1Motion=Free
  swing2LimitDegrees=45
  swing2Motion=Free
  twistLimitDegrees=45
  twistMotion=Free
 }
 {bone1Name=dyn_neck_midback_r, bone2Name=base_neck_midback_r, swing1LimitDegrees=5, swing2LimitDegrees=15, twistLimitDegrees=1}
 {
  bone1Name=spine_04
  bone2Name=spine_03
  swing1LimitDegrees=45
  swing1Motion=Free
  swing2LimitDegrees=45
  swing2Motion=Free
  twistLimitDegrees=45
  twistMotion=Free
 }
 {
  bone1Name=lowerarm_l
  bone2Name=upperarm_l
  swing1LimitDegrees=45
  swing1Motion=Free
  swing2LimitDegrees=45
  swing2Motion=Free
  twistLimitDegrees=45
  twistMotion=Free
 }
 {
  bone1Name=base_cape_help
  bone2Name=spine_05
  swing1LimitDegrees=45
  swing1Motion=Free
  swing2LimitDegrees=45
  swing2Motion=Free
  twistLimitDegrees=45
  twistMotion=Free
 }
]
