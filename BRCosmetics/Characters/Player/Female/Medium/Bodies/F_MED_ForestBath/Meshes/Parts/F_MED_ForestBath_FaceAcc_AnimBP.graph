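# Animation Blueprint for the ForestBath face accessory, driving its hair on the female medium body.
# The parent mesh pose passes through a Control Rig and front-hair AnimDynamics into a cached pose.
# Backpack state selects between two RigidBody branches: one adds weighted body constraints and bone offsets, the other layers cape, skydive-drift and parachute adjustments.
# The event graph supplies gravity override, leftward-movement, backbling-type and delay-timed parachute flags.
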AnimBlueprint /BRCosmetics/Characters/Player/Female/Medium/Bodies/F_MED_ForestBath/Meshes/Parts/F_MED_ForestBath_FaceAcc_AnimBP
target_skeleton=&/F_MED_ForestBath_FaceAcc_Skeleton
graphs [
 AnimGraph {nodes [
   n1 = RigidBody!(
    ComponentPose=Local To Component(LocalPose=Use cached pose 'CACHE_Base'())
    Alpha
    OverrideWorldGravity:Vector=Get GravityOverride()
    ExternalForce
   ) -> (Pose) {
    bClampLinearTranslationLimitToRefPose=true
    bOverrideWorldGravity=true
    overridePhysicsAsset=&/F_MED_ForestBath_FaceAcc_BackpackOrCapePhysics
    simulationSpace=BaseBoneSpace
    alphaBoolBlend {blendOption=Linear}
    baseBoneRef {boneName=spine_05}
    componentAppliedLinearAccClamp=<3000 3000 3000>
    componentLinearAccScale=<0.6 0.6 0.6>
    componentLinearVelScale=<0.6 0.6 0.6>
    simSpaceSettings {
     maxAngularAcceleration=1000
     maxAngularVelocity=3
     maxLinearAcceleration=700
     maxLinearVelocity=500
     velocityScaleZ=0
     worldAlpha=0.2
     externalLinearDragV=<0.2 0.2>
    }
   }
   n2 = "Constraint - dyn_hair_back_l_02 (PPPPPP)"!(
    ComponentPose=@n1
    Alpha=0.3
    ConstraintWeights_0=0.6926
    ConstraintWeights_1:float=0.2717
    ConstraintWeights_2:float=0.0238
    ConstraintWeights_3:float=0.0077
    ConstraintWeights_4:float=0.0022
    ConstraintWeights_5:float=0.002
   ) -> (Pose) {
    alphaBoolBlend {blendOption=Linear}
    boneToModify {boneName=dyn_hair_back_l_02}
    constraintSetup [
     {targetBone {boneName=spine_05}}
     {targetBone {boneName=clavicle_l}}
     {targetBone {boneName=neck_01}}
     {targetBone {boneName=neck_02}}
     {targetBone {boneName=upperarm_twist_01_l}}
     {targetBone {boneName=deform_upperarm_up_l}}
    ]
    constraintWeights [1,1,1,1,1,1]
   }
   n3 = "Constraint - dyn_hair_back_m_02 (PPPPP)"!(
    ComponentPose=@n2
    Alpha=0.3
    ConstraintWeights_0=0.9578
    ConstraintWeights_1:float=0.015
    ConstraintWeights_2:float=0.015
    ConstraintWeights_3:float=0.009
    ConstraintWeights_4:float=0.0032
   ) -> (Pose) {
    alphaBoolBlend {blendOption=Linear}
    boneToModify {boneName=dyn_hair_back_m_02}
    constraintSetup [
     {targetBone {boneName=spine_05}}
     {targetBone {boneName=clavicle_l}}
     {targetBone {boneName=clavicle_r}}
     {targetBone {boneName=neck_01}}
     {targetBone {boneName=neck_02}}
    ]
    constraintWeights [1,1,1,1,1]
   }
   n4 = "Constraint - dyn_hair_back_r_02 (PPPPPP)"!(
    ComponentPose=@n3
    Alpha=0.3
    ConstraintWeights_0=0.6926
    ConstraintWeights_1:float=0.2717
    ConstraintWeights_2:float=0.0238
    ConstraintWeights_3:float=0.0077
    ConstraintWeights_4:float=0.0022
    ConstraintWeights_5:float=0.002
   ) -> (Pose) {
    alphaBoolBlend {blendOption=Linear}
    boneToModify {boneName=dyn_hair_back_r_02}
    constraintSetup [
     {targetBone {boneName=spine_05}}
     {targetBone {boneName=clavicle_r}}
     {targetBone {boneName=neck_01}}
     {targetBone {boneName=neck_02}}
     {targetBone {boneName=upperarm_twist_01_r}}
     {targetBone {boneName=deform_upperarm_up_r}}
    ]
    constraintWeights [1,1,1,1,1,1]
   }
   n5 = "Constraint - dyn_hair_back_l_03 (PPPPPPPP)"!(
    ComponentPose=@n4
    Alpha=0.5
    ConstraintWeights_0=0.373
    ConstraintWeights_1:float=0.2288
    ConstraintWeights_2:float=0.1831
    ConstraintWeights_3:float=0.1354
    ConstraintWeights_4:float=0.0475
    ConstraintWeights_5:float=0.0301
    ConstraintWeights_6:float=0.001
    ConstraintWeights_7:float=0.001
   ) -> (Pose) {
    alphaBoolBlend {blendOption=Linear}
    boneToModify {boneName=dyn_hair_back_l_03}
    constraintSetup [
     {targetBone {boneName=clavicle_l}}
     {targetBone {boneName=spine_05}}
     {targetBone {boneName=upperarm_twist_01_l}}
     {targetBone {boneName=deform_upperarm_bk_l}}
     {targetBone {boneName=spine_04}}
     {targetBone {boneName=deform_upperarm_up_l}}
     {targetBone {boneName=spine_03}}
     {targetBone {boneName=spine_02}}
    ]
    constraintWeights [1,1,1,1,1,1,1,1]
   }
   n6 = "Constraint - dyn_hair_back_m_03 (PPPPPPPP)"!(
    ComponentPose=@n5
    Alpha=0.5
    ConstraintWeights_0=0.6809
    ConstraintWeights_1:float=0.1382
    ConstraintWeights_2:float=0.1382
    ConstraintWeights_3:float=0.0289
    ConstraintWeights_4:float=0.0042
    ConstraintWeights_5:float=0.0042
    ConstraintWeights_6:float=0.0027
    ConstraintWeights_7:float=0.0027
   ) -> (Pose) {
    alphaBoolBlend {blendOption=Linear}
    boneToModify {boneName=dyn_hair_back_m_03}
    constraintSetup [
     {targetBone {boneName=spine_05}}
     {targetBone {boneName=clavicle_l}}
     {targetBone {boneName=clavicle_r}}
     {targetBone {boneName=spine_04}}
     {targetBone {boneName=upperarm_twist_01_l}}
     {targetBone {boneName=upperarm_twist_01_r}}
     {targetBone {boneName=deform_upperarm_bk_l}}
     {targetBone {boneName=deform_upperarm_bk_r}}
    ]
    constraintWeights [1,1,1,1,1,1,1,1]
   }
   n7 = "Constraint - dyn_hair_back_r_03 (PPPPPPPP)"!(
    ComponentPose=@n6
    Alpha=0.5
    ConstraintWeights_0=0.373
    ConstraintWeights_1:float=0.2288
    ConstraintWeights_2:float=0.1831
    ConstraintWeights_3:float=0.1354
    ConstraintWeights_4:float=0.0475
    ConstraintWeights_5:float=0.0301
    ConstraintWeights_6:float=0.001
    ConstraintWeights_7:float=0.001
   ) -> (Pose) {
    alphaBoolBlend {blendOption=Linear}
    boneToModify {boneName=dyn_hair_back_r_03}
    constraintSetup [
     {targetBone {boneName=clavicle_r}}
     {targetBone {boneName=spine_05}}
     {targetBone {boneName=upperarm_twist_01_r}}
     {targetBone {boneName=deform_upperarm_bk_r}}
     {targetBone {boneName=spine_04}}
     {targetBone {boneName=deform_upperarm_up_r}}
     {targetBone {boneName=spine_03}}
     {targetBone {boneName=spine_02}}
    ]
    constraintWeights [1,1,1,1,1,1,1,1]
   }
   n8 = "Constraint - dyn_hair_back_l_04 (PPPPPPPP)"!(
    ComponentPose=@n7
    Alpha=0.7
    ConstraintWeights_0=0.2691
    ConstraintWeights_1:float=0.2339
    ConstraintWeights_2:float=0.1979
    ConstraintWeights_3:float=0.1711
    ConstraintWeights_4:float=0.0833
    ConstraintWeights_5:float=0.0213
    ConstraintWeights_6:float=0.0195
    ConstraintWeights_7:float=0.0039
   ) -> (Pose) {
    alphaBoolBlend {blendOption=Linear}
    boneToModify {boneName=dyn_hair_back_l_04}
    constraintSetup [
     {targetBone {boneName=spine_05}}
     {targetBone {boneName=clavicle_l}}
     {targetBone {boneName=spine_04}}
     {targetBone {boneName=upperarm_twist_01_l}}
     {targetBone {boneName=deform_upperarm_bk_l}}
     {targetBone {boneName=spine_02}}
     {targetBone {boneName=spine_03}}
     {targetBone {boneName=deform_upperarm_up_l}}
    ]
    constraintWeights [1,1,1,1,1,1,1,1]
   }
   n9 = "Constraint - dyn_hair_back_l_04 (PPPPPPPP)"(
    ComponentPose=@n8
    Alpha=0.7
    ConstraintWeights_0=0.2691
    ConstraintWeights_1=0.2339
    ConstraintWeights_2=0.1979
    ConstraintWeights_3=0.1711
    ConstraintWeights_4=0.0833
    ConstraintWeights_5=0.0213
    ConstraintWeights_6=0.0195
    ConstraintWeights_7=0.0039
   ) {
    alphaBoolBlend {blendOption=Linear}
    boneToModify {boneName=dyn_hair_back_l_04}
    constraintSetup [
     {targetBone {boneName=spine_05}}
     {targetBone {boneName=clavicle_l}}
     {targetBone {boneName=spine_04}}
     {targetBone {boneName=upperarm_twist_01_l}}
     {targetBone {boneName=deform_upperarm_bk_l}}
     {targetBone {boneName=spine_02}}
     {targetBone {boneName=spine_03}}
     {targetBone {boneName=deform_upperarm_up_l}}
    ]
    constraintWeights [1,1,1,1,1,1,1,1]
   }
   n10 = "Constraint - dyn_hair_back_r_04 (PPPPPPP)"!(
    ComponentPose=@n9
    Alpha=0.7
    ConstraintWeights_0=0.2664
    ConstraintWeights_1:float=0.2348
    ConstraintWeights_2:float=0.2229
    ConstraintWeights_3:float=0.1547
    ConstraintWeights_4:float=0.0713
    ConstraintWeights_5:float=0.026
    ConstraintWeights_6:float=0.0239
   ) -> (Pose) {
    alphaBoolBlend {blendOption=Linear}
    boneToModify {boneName=dyn_hair_back_r_04}
    constraintSetup [
     {targetBone {boneName=spine_05}}
     {targetBone {boneName=clavicle_r}}
     {targetBone {boneName=spine_04}}
     {targetBone {boneName=upperarm_twist_01_r}}
     {targetBone {boneName=deform_upperarm_bk_r}}
     {targetBone {boneName=spine_02}}
     {targetBone {boneName=spine_03}}
    ]
    constraintWeights [1,1,1,1,1,1,1]
   }
   n11 = "Transform (Modify) Bone - Bone: dyn_hair_back_l_01"!(ComponentPose=@n10, bAlphaBoolEnabled:bool=True, Translation="0,1,0", Rotation="0,0,0", Scale="1,0.8,1") -> (Pose) {
    alphaInputType=Bool
    rotationMode=BMM_Additive
    rotationSpace=BCS_BoneSpace
    scaleMode=BMM_Additive
    scaleSpace=BCS_BoneSpace
    translationMode=BMM_Additive
    translationSpace=BCS_BoneSpace
    alphaBoolBlend {blendOption=Linear}
    boneToModify {boneName=dyn_hair_back_l_01}
    scale {y=0.8}
    translation {y=1}
   }
   n12 = "Transform (Modify) Bone - Bone: dyn_hair_back_m_01"!(ComponentPose=@n11, bAlphaBoolEnabled:bool=True, Translation="0,1,0", Rotation="0,0,0", Scale="1,0.8,1") -> (Pose) {
    alphaInputType=Bool
    rotationMode=BMM_Additive
    rotationSpace=BCS_BoneSpace
    scaleMode=BMM_Additive
    scaleSpace=BCS_BoneSpace
    translationMode=BMM_Additive
    translationSpace=BCS_BoneSpace
    alphaBoolBlend {blendOption=Linear}
    boneToModify {boneName=dyn_hair_back_m_01}
    scale {y=0.8}
    translation {y=1}
   }
   n13 = "Transform (Modify) Bone - Bone: dyn_hair_back_r_01"!(ComponentPose=@n12, bAlphaBoolEnabled:bool=True, Translation="0,1,0", Rotation="0,0,0", Scale="1,0.8,1") -> (Pose) {
    alphaInputType=Bool
    rotationMode=BMM_Additive
    rotationSpace=BCS_BoneSpace
    scaleMode=BMM_Additive
    scaleSpace=BCS_BoneSpace
    translationMode=BMM_Additive
    translationSpace=BCS_BoneSpace
    alphaBoolBlend {blendOption=Linear}
    boneToModify {boneName=dyn_hair_back_r_01}
    scale {y=0.8}
    translation {y=1}
   }
   n14 = "Transform (Modify) Bone - Bone: dyn_hair_back_l_02"!(ComponentPose=@n13, bAlphaBoolEnabled:bool=True, Translation="0,-1,0", Rotation="0,0,0", Scale="1,1,1") -> (Pose) {
    alphaInputType=Bool
    rotationMode=BMM_Additive
    rotationSpace=BCS_BoneSpace
    scaleMode=BMM_Additive
    scaleSpace=BCS_BoneSpace
    translationMode=BMM_Additive
    translationSpace=BCS_BoneSpace
    alphaBoolBlend {blendOption=Linear}
    boneToModify {boneName=dyn_hair_back_l_02}
    translation {y=-1}
   }
   n15 = "Transform (Modify) Bone - Bone: dyn_hair_back_m_02"!(ComponentPose=@n14, bAlphaBoolEnabled:bool=True, Translation="0,-1,0", Rotation="0,0,0", Scale="1,1,1") -> (Pose) {
    alphaInputType=Bool
    rotationMode=BMM_Additive
    rotationSpace=BCS_BoneSpace
    scaleMode=BMM_Additive
    scaleSpace=BCS_BoneSpace
    translationMode=BMM_Additive
    translationSpace=BCS_BoneSpace
    alphaBoolBlend {blendOption=Linear}
    boneToModify {boneName=dyn_hair_back_m_02}
    translation {y=-1}
   }
   n16 = "Transform (Modify) Bone - Bone: dyn_hair_back_r_02"!(ComponentPose=@n15, bAlphaBoolEnabled:bool=True, Translation="0,-1,0", Rotation="0,0,0", Scale="1,1,1") -> (Pose) {
    alphaInputType=Bool
    rotationMode=BMM_Additive
    rotationSpace=BCS_BoneSpace
    scaleMode=BMM_Additive
    scaleSpace=BCS_BoneSpace
    translationMode=BMM_Additive
    translationSpace=BCS_BoneSpace
    alphaBoolBlend {blendOption=Linear}
    boneToModify {boneName=dyn_hair_back_r_02}
    translation {y=-1}
   }
   n17 = "Transform (Modify) Bone - Bone: dyn_hair_back_l_03"!(ComponentPose=@n16, bAlphaBoolEnabled:bool=True, Translation="0,0.5,0", Rotation="0,0,0", Scale="1,0.5,1") -> (Pose) {
    alphaInputType=Bool
    rotationMode=BMM_Additive
    rotationSpace=BCS_BoneSpace
    scaleMode=BMM_Additive
    scaleSpace=BCS_BoneSpace
    translationMode=BMM_Additive
    translationSpace=BCS_BoneSpace
    alphaBoolBlend {blendOption=Linear}
    boneToModify {boneName=dyn_hair_back_l_03}
    scale {y=0.5}
    translation {y=0.5}
   }
   n18 = "Transform (Modify) Bone - Bone: dyn_hair_back_m_03"!(ComponentPose=@n17, bAlphaBoolEnabled:bool=True, Translation="0,0.5,0", Rotation="0,0,0", Scale="1,0.5,1") -> (Pose) {
    alphaInputType=Bool
    rotationMode=BMM_Additive
    rotationSpace=BCS_BoneSpace
    scaleMode=BMM_Additive
    scaleSpace=BCS_BoneSpace
    translationMode=BMM_Additive
    translationSpace=BCS_BoneSpace
    alphaBoolBlend {blendOption=Linear}
    boneToModify {boneName=dyn_hair_back_m_03}
    scale {y=0.5}
    translation {y=0.5}
   }
   n19 = "Transform (Modify) Bone - Bone: dyn_hair_back_r_03"!(ComponentPose=@n18, bAlphaBoolEnabled:bool=True, Translation="0,0.5,0", Rotation="0,0,0", Scale="1,0.5,1") -> (Pose) {
    alphaInputType=Bool
    rotationMode=BMM_Additive
    rotationSpace=BCS_BoneSpace
    scaleMode=BMM_Additive
    scaleSpace=BCS_BoneSpace
    translationMode=BMM_Additive
    translationSpace=BCS_BoneSpace
    alphaBoolBlend {blendOption=Linear}
    boneToModify {boneName=dyn_hair_back_r_03}
    scale {y=0.5}
    translation {y=0.5}
   }
   n20 = "Transform (Modify) Bone - Bone: dyn_hair_back_l_01"(ComponentPose=@n19, bAlphaBoolEnabled=True, Translation="0,1,0", Rotation="0,0,0", Scale="1,1,1") {
    alphaInputType=Bool
    rotationMode=BMM_Additive
    rotationSpace=BCS_BoneSpace
    scaleMode=BMM_Additive
    scaleSpace=BCS_BoneSpace
    translationMode=BMM_Additive
    translationSpace=BCS_BoneSpace
    alphaBoolBlend {blendOption=Linear}
    boneToModify {boneName=dyn_hair_back_l_01}
    translation {y=1}
   }
   n21 = "Transform (Modify) Bone - Bone: dyn_hair_back_m_01"(ComponentPose=@n20, bAlphaBoolEnabled=True, Translation="0,1,0", Rotation="0,0,0", Scale="1,1,1") {
    alphaInputType=Bool
    rotationMode=BMM_Additive
    rotationSpace=BCS_BoneSpace
    scaleMode=BMM_Additive
    scaleSpace=BCS_BoneSpace
    translationMode=BMM_Additive
    translationSpace=BCS_BoneSpace
    alphaBoolBlend {blendOption=Linear}
    boneToModify {boneName=dyn_hair_back_m_01}
    translation {y=1}
   }
   n22 = "Transform (Modify) Bone - Bone: dyn_hair_back_r_01"(ComponentPose=@n21, bAlphaBoolEnabled=True, Translation="0,1,0", Rotation="0,0,0", Scale="1,1,1") {
    alphaInputType=Bool
    rotationMode=BMM_Additive
    rotationSpace=BCS_BoneSpace
    scaleMode=BMM_Additive
    scaleSpace=BCS_BoneSpace
    translationMode=BMM_Additive
    translationSpace=BCS_BoneSpace
    alphaBoolBlend {blendOption=Linear}
    boneToModify {boneName=dyn_hair_back_r_01}
    translation {y=1}
   }
   n23 = "Transform (Modify) Bone - Bone: dyn_hair_back_l_02"(ComponentPose=@n22, bAlphaBoolEnabled=True, Translation="0,3,0", Rotation="0,0,0", Scale="1,0.8,1") {
    alphaInputType=Bool
    rotationMode=BMM_Additive
    rotationSpace=BCS_BoneSpace
    scaleMode=BMM_Additive
    scaleSpace=BCS_BoneSpace
    translationMode=BMM_Additive
    translationSpace=BCS_BoneSpace
    alphaBoolBlend {blendOption=Linear}
    boneToModify {boneName=dyn_hair_back_l_02}
    scale {y=0.8}
    translation {y=3}
   }
   n24 = "Transform (Modify) Bone - Bone: dyn_hair_back_m_02"(ComponentPose=@n23, bAlphaBoolEnabled=True, Translation="0,3,0", Rotation="0,0,0", Scale="1,0.8,1") {
    alphaInputType=Bool
    rotationMode=BMM_Additive
    rotationSpace=BCS_BoneSpace
    scaleMode=BMM_Additive
    scaleSpace=BCS_BoneSpace
    translationMode=BMM_Additive
    translationSpace=BCS_BoneSpace
    alphaBoolBlend {blendOption=Linear}
    boneToModify {boneName=dyn_hair_back_m_02}
    scale {y=0.8}
    translation {y=3}
   }
   n25 = "Transform (Modify) Bone - Bone: dyn_hair_back_r_02"(ComponentPose=@n24, bAlphaBoolEnabled=True, Translation="0,3,0", Rotation="0,0,0", Scale="1,0.8,1") {
    alphaInputType=Bool
    rotationMode=BMM_Additive
    rotationSpace=BCS_BoneSpace
    scaleMode=BMM_Additive
    scaleSpace=BCS_BoneSpace
    translationMode=BMM_Additive
    translationSpace=BCS_BoneSpace
    alphaBoolBlend {blendOption=Linear}
    boneToModify {boneName=dyn_hair_back_r_02}
    scale {y=0.8}
    translation {y=3}
   }
   n26 = "Transform (Modify) Bone - Bone: dyn_hair_back_l_03"(ComponentPose=@n25, bAlphaBoolEnabled=True, Translation="0,5,0", Rotation="0,0,0", Scale="1,1,1") {
    alphaInputType=Bool
    rotationMode=BMM_Additive
    rotationSpace=BCS_BoneSpace
    scaleMode=BMM_Additive
    scaleSpace=BCS_BoneSpace
    translationMode=BMM_Additive
    translationSpace=BCS_BoneSpace
    alphaBoolBlend {blendOption=Linear}
    boneToModify {boneName=dyn_hair_back_l_03}
    translation {y=5}
   }
   n27 = "Transform (Modify) Bone - Bone: dyn_hair_back_m_03"(ComponentPose=@n26, bAlphaBoolEnabled=True, Translation="0,5,0", Rotation="0,0,0", Scale="1,1,1") {
    alphaInputType=Bool
    rotationMode=BMM_Additive
    rotationSpace=BCS_BoneSpace
    scaleMode=BMM_Additive
    scaleSpace=BCS_BoneSpace
    translationMode=BMM_Additive
    translationSpace=BCS_BoneSpace
    alphaBoolBlend {blendOption=Linear}
    boneToModify {boneName=dyn_hair_back_m_03}
    translation {y=5}
   }
   n28 = "Transform (Modify) Bone - Bone: dyn_hair_back_r_03"(ComponentPose=@n27, bAlphaBoolEnabled=True, Translation="0,5,0", Rotation="0,0,0", Scale="1,1,1") {
    alphaInputType=Bool
    rotationMode=BMM_Additive
    rotationSpace=BCS_BoneSpace
    scaleMode=BMM_Additive
    scaleSpace=BCS_BoneSpace
    translationMode=BMM_Additive
    translationSpace=BCS_BoneSpace
    alphaBoolBlend {blendOption=Linear}
    boneToModify {boneName=dyn_hair_back_r_03}
    translation {y=5}
   }
   n29 = Get ParachutClossing() -> (ParachutClossing:bool)
   n30 = Get ParachutBeginning() -> (ParachutBeginning:bool)
   n31 = "Transform (Modify) Bone - Bone: base_cape_help"(
    ComponentPose=Local To Component(LocalPose=Use cached pose 'CACHE_Base'())
    Alpha=1
    Translation="5.5,6.5,0"
    Rotation="0,0,0"
    Scale="1,1,1"
   ) {
    rotationMode=BMM_Additive
    rotationSpace=BCS_BoneSpace
    scaleMode=BMM_Additive
    scaleSpace=BCS_BoneSpace
    translationMode=BMM_Additive
    translationSpace=BCS_BoneSpace
    alphaBoolBlend {blendInTime=0.5, blendOption=Sinusoidal, blendOutTime=1}
    boneToModify {boneName=base_cape_help}
    translation=<5.5 6.5>
   }
   n32 = "Transform (Modify) Bone - Bone: base_cape_help"!(ComponentPose=@n31, bAlphaBoolEnabled:bool=True, Translation="-4,4,0", Rotation="0,0,0", Scale="1,1,1") -> (Pose) {
    alphaInputType=Bool
    rotationMode=BMM_Additive
    rotationSpace=BCS_BoneSpace
    scaleMode=BMM_Additive
    scaleSpace=BCS_BoneSpace
    translationMode=BMM_Additive
    translationSpace=BCS_BoneSpace
    alphaBoolBlend {blendInTime=0.5, blendOption=Sinusoidal, blendOutTime=1}
    boneToModify {boneName=base_cape_help}
    translation=<-4 4>
   }
   n33 = Component To Local(ComponentPose=@n32)
   n34 = "Transform (Modify) Bone - Bone: base_cape_help"!(
    ComponentPose=Local To Component(LocalPose=Use cached pose 'CACHE_Base'())
    Alpha:float=1
    Translation="-30,0,0"
    Rotation="0,0,0"
    Scale="1,1,1"
   ) -> (Pose) {
    rotationMode=BMM_Additive
    rotationSpace=BCS_BoneSpace
    scaleMode=BMM_Additive
    scaleSpace=BCS_BoneSpace
    translationMode=BMM_Additive
    translationSpace=BCS_BoneSpace
    alphaBoolBlend {blendInTime=0.5, blendOption=Sinusoidal, blendOutTime=1}
    boneToModify {boneName=base_cape_help}
    translation {x=-30}
   }
   n35 = Component To Local(ComponentPose=@n34)
   n36 = Get Is Wearing Cape() -> (Is Wearing Cape:bool)
   n37 = "Transform (Modify) Bone - Bone: base_hair_back"!(
    ComponentPose=Local To Component(LocalPose=Blend Poses by bool(BlendPose_0=@n33, BlendPose_1=@n35, bActiveValue=@n36) {blendType=Linear})
    bAlphaBoolEnabled:bool=Get SkyDiveFloating_MoveLeft() -> (SkyDiveFloating_MoveLeft:bool)
    Translation="0,0,10"
    Rotation="0,0,15"
    Scale="1,1,1"
   ) -> (Pose) {
    alphaInputType=Bool
    rotationMode=BMM_Additive
    rotationSpace=BCS_BoneSpace
    scaleMode=BMM_Additive
    scaleSpace=BCS_BoneSpace
    translationMode=BMM_Additive
    translationSpace=BCS_BoneSpace
    alphaBoolBlend {blendInTime=0.5, blendOption=Sinusoidal, blendOutTime=1}
    boneToModify {boneName=base_hair_back}
    rotation {roll=15}
    translation {z=10}
   }
   n38 = RigidBody(ComponentPose=@n37, OverrideWorldGravity=Get GravityOverride()) {
    bClampLinearTranslationLimitToRefPose=true
    bOverrideWorldGravity=true
    overridePhysicsAsset=&/F_MED_ForestBath_FaceAcc_Physics
    simulationSpace=BaseBoneSpace
    alphaBoolBlend {blendOption=Linear}
    baseBoneRef {boneName=spine_05}
    componentAppliedLinearAccClamp=<3000 3000 3000>
    componentLinearAccScale=<0.6 0.6 0.6>
    componentLinearVelScale=<0.6 0.6 0.6>
    simSpaceSettings {
     maxAngularAcceleration=1000
     maxAngularVelocity=3
     maxLinearAcceleration=700
     maxLinearVelocity=500
     velocityScaleZ=0
     worldAlpha=0.2
     externalLinearDragV=<0.2 0.2>
    }
   }
   n39 = "Transform (Modify) Bone - Bone: base_hair_back_l"!(ComponentPose=@n38, bAlphaBoolEnabled:bool=True, Translation="0,0,0", Rotation="0,-10,0", Scale="1,0.7,1") -> (Pose) {
    alphaInputType=Bool
    rotationMode=BMM_Additive
    rotationSpace=BCS_BoneSpace
    scaleMode=BMM_Additive
    scaleSpace=BCS_BoneSpace
    translationMode=BMM_Additive
    translationSpace=BCS_BoneSpace
    alphaBoolBlend {blendInTime=0.5, blendOption=Sinusoidal, blendOutTime=1}
    boneToModify {boneName=base_hair_back_l}
    rotation {yaw=-10}
    scale {y=0.7}
   }
   n40 = "Transform (Modify) Bone - Bone: base_hair_back_m"!(ComponentPose=@n39, bAlphaBoolEnabled:bool=True, Translation="0,0,0", Rotation="0,-10,0", Scale="1,0.7,1") -> (Pose) {
    alphaInputType=Bool
    rotationMode=BMM_Additive
    rotationSpace=BCS_BoneSpace
    scaleMode=BMM_Additive
    scaleSpace=BCS_BoneSpace
    translationMode=BMM_Additive
    translationSpace=BCS_BoneSpace
    alphaBoolBlend {blendInTime=0.5, blendOption=Sinusoidal, blendOutTime=1}
    boneToModify {boneName=base_hair_back_m}
    rotation {yaw=-10}
    scale {y=0.7}
   }
   n41 = "Transform (Modify) Bone - Bone: base_hair_back_r"!(ComponentPose=@n40, bAlphaBoolEnabled:bool=True, Translation="0,0,0", Rotation="0,-10,0", Scale="1,0.7,1") -> (Pose) {
    alphaInputType=Bool
    rotationMode=BMM_Additive
    rotationSpace=BCS_BoneSpace
    scaleMode=BMM_Additive
    scaleSpace=BCS_BoneSpace
    translationMode=BMM_Additive
    translationSpace=BCS_BoneSpace
    alphaBoolBlend {blendInTime=0.5, blendOption=Sinusoidal, blendOutTime=1}
    boneToModify {boneName=base_hair_back_r}
    rotation {yaw=-10}
    scale {y=0.7}
   }
   n42 = "Transform (Modify) Bone - Bone: base_hair_back_l"(ComponentPose=@n41, bAlphaBoolEnabled=@n30, Translation="0,0,0", Rotation="0,-45,0", Scale="1,0.7,1") {
    alphaInputType=Bool
    rotationMode=BMM_Additive
    rotationSpace=BCS_BoneSpace
    scaleMode=BMM_Additive
    scaleSpace=BCS_BoneSpace
    translationMode=BMM_Additive
    translationSpace=BCS_BoneSpace
    alphaBoolBlend {blendInTime=0.5, blendOption=Sinusoidal, blendOutTime=1}
    boneToModify {boneName=base_hair_back_l}
    rotation {yaw=-45}
    scale {y=0.7}
   }
   n43 = "Transform (Modify) Bone - Bone: base_hair_back_m"(ComponentPose=@n42, bAlphaBoolEnabled=@n30, Translation="0,0,0", Rotation="0,-45,0", Scale="1,0.7,1") {
    alphaInputType=Bool
    rotationMode=BMM_Additive
    rotationSpace=BCS_BoneSpace
    scaleMode=BMM_Additive
    scaleSpace=BCS_BoneSpace
    translationMode=BMM_Additive
    translationSpace=BCS_BoneSpace
    alphaBoolBlend {blendInTime=0.5, blendOption=Sinusoidal, blendOutTime=1}
    boneToModify {boneName=base_hair_back_m}
    rotation {yaw=-45}
    scale {y=0.7}
   }
   n44 = "Transform (Modify) Bone - Bone: base_hair_back_r"(ComponentPose=@n43, bAlphaBoolEnabled=@n30, Translation="0,0,0", Rotation="0,-45,0", Scale="1,0.7,1") {
    alphaInputType=Bool
    rotationMode=BMM_Additive
    rotationSpace=BCS_BoneSpace
    scaleMode=BMM_Additive
    scaleSpace=BCS_BoneSpace
    translationMode=BMM_Additive
    translationSpace=BCS_BoneSpace
    alphaBoolBlend {blendInTime=0.5, blendOption=Sinusoidal, blendOutTime=1}
    boneToModify {boneName=base_hair_back_r}
    rotation {yaw=-45}
    scale {y=0.7}
   }
   n45 = "Transform (Modify) Bone - Bone: base_hair_back_l"(ComponentPose=@n44, bAlphaBoolEnabled=@n29, Translation="0,0,0", Rotation="0,-60,0", Scale="1,0.7,1") {
    alphaInputType=Bool
    rotationMode=BMM_Additive
    rotationSpace=BCS_BoneSpace
    scaleMode=BMM_Additive
    scaleSpace=BCS_BoneSpace
    translationMode=BMM_Additive
    translationSpace=BCS_BoneSpace
    alphaBoolBlend {blendOption=Sinusoidal, blendOutTime=1}
    boneToModify {boneName=base_hair_back_l}
    rotation {yaw=-60}
    scale {y=0.7}
   }
   n46 = "Transform (Modify) Bone - Bone: base_hair_back_m"(ComponentPose=@n45, bAlphaBoolEnabled=@n29, Translation="0,0,0", Rotation="0,-60,0", Scale="1,0.7,1") {
    alphaInputType=Bool
    rotationMode=BMM_Additive
    rotationSpace=BCS_BoneSpace
    scaleMode=BMM_Additive
    scaleSpace=BCS_BoneSpace
    translationMode=BMM_Additive
    translationSpace=BCS_BoneSpace
    alphaBoolBlend {blendOption=Sinusoidal, blendOutTime=1}
    boneToModify {boneName=base_hair_back_m}
    rotation {yaw=-60}
    scale {y=0.7}
   }
   n47 = "Transform (Modify) Bone - Bone: base_hair_back_r"(ComponentPose=@n46, bAlphaBoolEnabled=@n29, Translation="0,0,0", Rotation="0,-60,0", Scale="1,0.7,1") {
    alphaInputType=Bool
    rotationMode=BMM_Additive
    rotationSpace=BCS_BoneSpace
    scaleMode=BMM_Additive
    scaleSpace=BCS_BoneSpace
    translationMode=BMM_Additive
    translationSpace=BCS_BoneSpace
    alphaBoolBlend {blendOption=Sinusoidal, blendOutTime=1}
    boneToModify {boneName=base_hair_back_r}
    rotation {yaw=-60}
    scale {y=0.7}
   }
   Output Pose(Result=Blend Poses by bool(
     BlendPose_0=Component To Local(ComponentPose=@n28)
     BlendPose_1=Component To Local(ComponentPose=@n47)
     bActiveValue=Get isWearingBackpack() -> (isWearingBackpack:bool)
    ) {blendType=Linear}) {layerGroup=None}
   n48 = Control Rig!(
    Source=Copy Pose From Mesh() {bUseAttachedParent=true}
    Alpha
    IsCrouching:bool=Get bIsCrouching()
    IsMoving2D:bool=Get bIsMoving2D()
    IsTargeting:bool=Get bIsTargeting()
    WearingBackpack:bool=Get isWearingBackpack()
    WearingCape:bool=Get Is Wearing Cape()
   ) -> (Pose) {
    alphaBoolBlend {blendOption=Linear}
    controlRigAssetReference {blueprintRigClass=&/F_MED_ForestBath_FaceAcc_CtrlRig.F_MED_ForestBath_FaceAcc_CtrlRig_C}
   }
   n49 = Anim Dynamics - Bone: dyn_hair_top_front_r(ComponentPose=Local To Component(LocalPose=@n48), Alpha=1, ExternalForce="0,0,0") {
    angularDampingOverride=1
    angularSpringConstant=20000
    bAngularSpring=true
    bOverrideAngularDamping=true
    bUseGravityOverride=true
    gravityScale=0.35
    simulationSpace=RootRelative
    alphaBoolBlend {blendOption=Linear}
    boundBone {boneName=dyn_hair_top_front_r}
    componentLinearAccScale=<1 1 1>
    componentLinearVelScale=<1 1 1>
    gravityOverride {z=-980}
    physicsBodyDefinitions [{
      boundBone {boneName=dyn_hair_top_front_r}
      boxExtents=<20 30 30>
      constraintSetup {angularLimitsMax=<10 10 10>, angularLimitsMin=<-10 -10 -10>, angularTarget=<1 0 0>}
      localJointOffset=<4 0 0>
     }]
   }
   n50 = Anim Dynamics - Bone: dyn_hair_top_front_l(ComponentPose=@n49, Alpha=1, ExternalForce="0,0,0") {
    angularDampingOverride=1
    angularSpringConstant=20000
    bAngularSpring=true
    bOverrideAngularDamping=true
    bUseGravityOverride=true
    gravityScale=0.35
    simulationSpace=RootRelative
    alphaBoolBlend {blendOption=Linear}
    boundBone {boneName=dyn_hair_top_front_l}
    componentLinearAccScale=<1 1 1>
    componentLinearVelScale=<1 1 1>
    gravityOverride {z=-980}
    physicsBodyDefinitions [{
      boundBone {boneName=dyn_hair_top_front_l}
      boxExtents=<20 30 30>
      constraintSetup {angularLimitsMax=<10 10 10>, angularLimitsMin=<-10 -10 -10>, angularTarget=<1 0 0>}
      localJointOffset=<4 0 0>
     }]
   }
   Save cached pose 'CACHE_Base'(Pose=Component To Local(ComponentPose=@n50)) {cacheName=CACHE_Base}
  ]}
 EventGraph {nodes [
   Event BlueprintUpdateAnimation()
   > n1 = Sequence!(execute) -> (then_0, then_1, then_2:Exec)
   > n2 = Gravity Override State Function DEPRECATED(CharacterPartAnimInstance=Self-Reference(), GameTimeInSeconds=GetGameTimeInSeconds(), GravityOverrideParams=Get Gravity Override Params())
   > Set GravityOverride(GravityOverride=@n2)
   n3 = PlayerMovement(execute=@n1.then_1, AnimInstance=Self-Reference(), XDotProductCoefficient=100, YDotProductCoefficient=100, ZDotProductCoefficient=100)
   n4 = MakeVector(X=@n3, Y=@n3.YMovement, Z=@n3.ZMovement)
   > n5 = Set Move_Left(Move_Left:bool="float < float"(A=BreakVector(InVec=@n4), B=-50)) -> (Output_Get:bool)
   > Set SkyDiveFloating_MoveLeft(SkyDiveFloating_MoveLeft:bool=AND Boolean(A=@n5, B=Get bIsSkydiveFloating())) -> (Output_Get:bool)
   Event OnCharacterCustomizationLoaded()
   > n6 = Cast To PlayerPawn_Athena(Object=TryGetPawnOwner())
   > Set Body_Mesh(Body_Mesh=GetSkeletalMeshForPartType(self=@n6))
   n7 = CheckForBackbling()
   > Set IsWearingBackpackOrCape(IsWearingBackpackOrCape:bool=OR Boolean!(A=@n7, B=@n7.bIsBackblingACape, C:bool=@n7.bIsBackblingLong) -> (ReturnValue)) -> (Output_Get:bool)
   > Set isWearingBackpack(isWearingBackpack:bool=AND Boolean!(A=@n7, B=NOT Boolean(A=@n7.bIsBackblingACape), C:bool=NOT Boolean(A=@n7.bIsBackblingLong)) -> (ReturnValue)) -> (Output_Get:bool)
   > Set Is Wearing Cape(Is Wearing Cape=OR Boolean!(A=@n7.bIsBackblingACape, B=@n7.bIsBackblingLong) -> (ReturnValue))
   Event BlueprintInitializeAnimation()
   > Set Skel Mesh(Skel Mesh=GetOwnerSkeletalMeshComponent())
   Event OnParachuteOpened()
   > Set ParachutBeginning(ParachutBeginning:bool=true) -> (Output_Get:bool)
   > Delay(Duration=2)
   > Set ParachutBeginning(ParachutBeginning=false)
   Event OnParachuteClosed()
   > Set ParachutClossing(ParachutClossing:bool=true) -> (Output_Get:bool)
   > Delay(Duration=0.8)
   > Set ParachutClossing(ParachutClossing=false)
  ]}
]
variables [
 Body_Mesh {type=SkeletalMeshComponent}
 Gravity Override Params {type=GravityOverrideParamsStruct}
 GravityOverride {type=Vector}
 Is Wearing Cape {type=boolean}
 isWearingBackbling {type=boolean}
 isWearingBackpack {type=boolean}
 IsWearingBackpackOrCape {type=boolean}
 Move_Left {type=boolean}
 MovingCrouch_NoTarget {type=boolean}
 ParachutBeginning {type=boolean}
 ParachutClossing {type=boolean}
 Skel Mesh {type=SkeletalMeshComponent}
 SkyDiveFloating_MoveLeft {type=boolean}
]
